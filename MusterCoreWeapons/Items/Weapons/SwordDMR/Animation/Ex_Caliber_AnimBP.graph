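# Animation Blueprint for the Ex_Caliber rifle (SwordDMR) that poses its cylinder by ammo count.
# DefaultSlot animations over ref pose get a mesh-space per-bone blend on the cylinder bone, weighted by CylinderMaskAlpha.
# The CylinderOverride state machine enters via a conduit choosing two, one or full-ammo pose, then cycles through one-shot shell animations with zero crossfade, returning to full after the last.
# Driving variables are not declared here, and the EventGraph is empty.
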AnimBlueprint /MusterCoreWeapons/Items/Weapons/SwordDMR/Animation/Ex_Caliber_AnimBP
target_skeleton=/MusterCoreWeapons/Items/Weapons/SwordDMR/Meshes/Ex_Caliber_Rifle_Skeleton
graphs [
 AnimGraph {nodes [
   n1 = Local Space Ref Pose()
   AnimGraphNode_StateMachine_0 = CylinderOverride() {is=title AnimGraphNode_StateMachine, bAllowConduitEntryStates=true}
   Output Pose(Result=Layered blend per bone(
     BasePose=Slot 'DefaultSlot'(Source=@n1)
     BlendPoses_0=@AnimGraphNode_StateMachine_0
     BlendWeights_0=Get CylinderMaskAlpha() -> (CylinderMaskAlpha:float)
    ) {bMeshSpaceRotationBlend=true, layerSetup [{branchFilters [{boneName=cylinder}]}]}) {layerGroup=None}
  ]}
 CylinderOverride {
  entry=Conduit
  path=AnimGraph.AnimGraphNode_StateMachine_0.CylinderOverride
  ids=id
  nodes [
   AnimStateConduitNode_0 = Conduit()
   AnimStateNode_0 = FullAmmoPose() {is=title AnimStateNode}
   AnimStateNode_1 = AmmoFromThreeToTwo() {is=title AnimStateNode}
   AnimStateNode_2 = TwoAmmoPose() {is=title AnimStateNode}
   AnimStateNode_3 = AmmoFromTwoToOne() {is=title AnimStateNode}
   AnimStateNode_4 = AmmoFromOneToZero() {is=title AnimStateNode}
   AnimStateNode_5 = OneAmmoPose() {is=title AnimStateNode}
   AnimStateTransitionNode_0 = AnimStateTransitionNode() {crossfadeDuration=0}
   AnimStateTransitionNode_1 = AnimStateTransitionNode() {bAutomaticRuleBasedOnSequencePlayerInState=true, crossfadeDuration=0}
   AnimStateTransitionNode_2 = AnimStateTransitionNode() {priorityOrder=100}
   AnimStateTransitionNode_3 = AnimStateTransitionNode()
   AnimStateTransitionNode_4 = AnimStateTransitionNode() {priorityOrder=2}
   AnimStateTransitionNode_5 = AnimStateTransitionNode() {crossfadeDuration=0}
   AnimStateTransitionNode_6 = AnimStateTransitionNode() {bAutomaticRuleBasedOnSequencePlayerInState=true, crossfadeDuration=0}
   AnimStateTransitionNode_7 = AnimStateTransitionNode() {crossfadeDuration=0}
   AnimStateTransitionNode_8 = AnimStateTransitionNode() {bAutomaticRuleBasedOnSequencePlayerInState=true, crossfadeDuration=0}
  ]
  links [
   {
    dst=FullAmmoPose
    rule=AnimGraph.AnimGraphNode_StateMachine_0.CylinderOverride.AnimStateTransitionNode_8.Transition
    src=AmmoFromOneToZero
    via=AnimStateTransitionNode_8
   }
   {
    dst=TwoAmmoPose
    rule=AnimGraph.AnimGraphNode_StateMachine_0.CylinderOverride.AnimStateTransitionNode_1.Transition
    src=AmmoFromThreeToTwo
    via=AnimStateTransitionNode_1
   }
   {
    dst=OneAmmoPose
    rule=AnimGraph.AnimGraphNode_StateMachine_0.CylinderOverride.AnimStateTransitionNode_6.Transition
    src=AmmoFromTwoToOne
    via=AnimStateTransitionNode_6
   }
   {
    dst=FullAmmoPose
    rule=AnimGraph.AnimGraphNode_StateMachine_0.CylinderOverride.AnimStateTransitionNode_2.Transition
    src=Conduit
    via=AnimStateTransitionNode_2
   }
   {
    dst=OneAmmoPose
    rule=AnimGraph.AnimGraphNode_StateMachine_0.CylinderOverride.AnimStateTransitionNode_4.Transition
    src=Conduit
    via=AnimStateTransitionNode_4
   }
   {
    dst=TwoAmmoPose
    rule=AnimGraph.AnimGraphNode_StateMachine_0.CylinderOverride.AnimStateTransitionNode_3.Transition
    src=Conduit
    via=AnimStateTransitionNode_3
   }
   {
    dst=AmmoFromThreeToTwo
    rule=AnimGraph.AnimGraphNode_StateMachine_0.CylinderOverride.AnimStateTransitionNode_0.Transition
    src=FullAmmoPose
    via=AnimStateTransitionNode_0
   }
   {
    dst=AmmoFromOneToZero
    rule=AnimGraph.AnimGraphNode_StateMachine_0.CylinderOverride.AnimStateTransitionNode_7.Transition
    src=OneAmmoPose
    via=AnimStateTransitionNode_7
   }
   {
    dst=AmmoFromTwoToOne
    rule=AnimGraph.AnimGraphNode_StateMachine_0.CylinderOverride.AnimStateTransitionNode_5.Transition
    src=TwoAmmoPose
    via=AnimStateTransitionNode_5
   }
  ]
 }
 Conduit {path=AnimGraph.AnimGraphNode_StateMachine_0.CylinderOverride.AnimStateConduitNode_0.Conduit, nodes [Result(bCanEnterTransition=true)]}
 FullAmmoPose {
  path=AnimGraph.AnimGraphNode_StateMachine_0.CylinderOverride.AnimStateNode_0.FullAmmoPose
  nodes [Output Animation Pose(Result=Sequence Player 'Ex_Caliber_Fire_Shell1_POSE_WEP'() {sequence=&/Ex_Caliber_Fire_Shell1_POSE_WEP})]
 }
 AmmoFromThreeToTwo {
  path=AnimGraph.AnimGraphNode_StateMachine_0.CylinderOverride.AnimStateNode_1.AmmoFromThreeToTwo
  nodes [Output Animation Pose(Result=Sequence Player 'Ex_Caliber_Fire_Shell1to2_WEP'() {bLoopAnimation=false, sequence=&/Ex_Caliber_Fire_Shell1to2_WEP})]
 }
 TwoAmmoPose {
  path=AnimGraph.AnimGraphNode_StateMachine_0.CylinderOverride.AnimStateNode_2.TwoAmmoPose
  nodes [Output Animation Pose(Result=Sequence Player 'Ex_Caliber_Fire_Shell2_POSE_WEP'() {sequence=&/Ex_Caliber_Fire_Shell2_POSE_WEP})]
 }
 AmmoFromTwoToOne {
  path=AnimGraph.AnimGraphNode_StateMachine_0.CylinderOverride.AnimStateNode_3.AmmoFromTwoToOne
  nodes [Output Animation Pose(Result=Sequence Player 'Ex_Caliber_Fire_Shell2to3_WEP'() {bLoopAnimation=false, sequence=&/Ex_Caliber_Fire_Shell2to3_WEP})]
 }
 AmmoFromOneToZero {
  path=AnimGraph.AnimGraphNode_StateMachine_0.CylinderOverride.AnimStateNode_4.AmmoFromOneToZero
  nodes [Output Animation Pose(Result=Sequence Player 'Ex_Caliber_Fire_Shell3to1_WEP'() {bLoopAnimation=false, sequence=&/Ex_Caliber_Fire_Shell3to1_WEP})]
 }
 OneAmmoPose {
  path=AnimGraph.AnimGraphNode_StateMachine_0.CylinderOverride.AnimStateNode_5.OneAmmoPose
  nodes [Output Animation Pose(Result=Sequence Player 'Ex_Caliber_Fire_Shell3_Pose_WEP'() {sequence=&/Ex_Caliber_Fire_Shell3_Pose_WEP})]
 }
 Transition {
  path=AnimGraph.AnimGraphNode_StateMachine_0.CylinderOverride.AnimStateTransitionNode_0.Transition
  nodes [Result(bCanEnterTransition=Get bStateRuleFromFullToTwoAmmo() -> (bStateRuleFromFullToTwoAmmo:bool))]
 }
 Transition {path=AnimGraph.AnimGraphNode_StateMachine_0.CylinderOverride.AnimStateTransitionNode_1.Transition, nodes [Result()]}
 Transition {path=AnimGraph.AnimGraphNode_StateMachine_0.CylinderOverride.AnimStateTransitionNode_2.Transition, nodes [Result(bCanEnterTransition=true)]}
 Transition {
  path=AnimGraph.AnimGraphNode_StateMachine_0.CylinderOverride.AnimStateTransitionNode_3.Transition
  nodes [Result(bCanEnterTransition=Get bIsAmmoCountTwo() -> (bIsAmmoCountTwo:bool))]
 }
 Transition {
  path=AnimGraph.AnimGraphNode_StateMachine_0.CylinderOverride.AnimStateTransitionNode_4.Transition
  nodes [Result(bCanEnterTransition=Get bIsAmmoCountOne() -> (bIsAmmoCountOne:bool))]
 }
 Transition {
  path=AnimGraph.AnimGraphNode_StateMachine_0.CylinderOverride.AnimStateTransitionNode_5.Transition
  nodes [Result(bCanEnterTransition=Get bStateRuleFromTwoToOneAmmo() -> (bStateRuleFromTwoToOneAmmo:bool))]
 }
 Transition {path=AnimGraph.AnimGraphNode_StateMachine_0.CylinderOverride.AnimStateTransitionNode_6.Transition, nodes [Result()]}
 Transition {
  path=AnimGraph.AnimGraphNode_StateMachine_0.CylinderOverride.AnimStateTransitionNode_7.Transition
  nodes [Result(bCanEnterTransition=Get bStateRuleFromOneToZero() -> (bStateRuleFromOneToZero:bool))]
 }
 Transition {path=AnimGraph.AnimGraphNode_StateMachine_0.CylinderOverride.AnimStateTransitionNode_8.Transition, nodes [Result()]}
 EventGraph {links [], nodes []}
]
variables []
