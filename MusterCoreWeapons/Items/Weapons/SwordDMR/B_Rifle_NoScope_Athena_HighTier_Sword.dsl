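; Cosmetic weapon Blueprint for a high-tier sword-firing DMR (Excalibur), driving the sword's visibility, glow and Niagara effects.
; Firing hides the sword and idle effect, then SwordRezIn restores them after a delay if enough ammo remains, animating reveal and Fresnel through a timeline.
; Reload stages swap emissive-color timelines and re-materialize the sword on cartridge load.
; Weapon visibility changes toggle the idle effect and replay a Fresnel fade when shown.
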(event EventOnWeaponVisibilityChanged
  (params (bVisible :type Boolean) (bSetForLocalControllerOnly :type Boolean))
  (|CallParentFunction bVisible bSetForLocalControllerOnly)
  (if bVisible
    (Rendering|Material|SetScalarParameterValueonMaterials (Variables|Visual|GetWeaponMesh) "Opacity" 1.0)
    (Components|Activation|Activate (Variables|Default|GetNS_ExCalibur_Idle) true)
    (bind (_direction _newtrack_0) (Timeline|Timeline_1 PlayFromStart)
      (:Update
        (Rendering|Material|SetScalarParameterValueonMaterials (Variables|Default|GetFXStaticMeshComponent) "Fresnel Brightness" (* _newtrack_0 0.15))))
    (else
      (Components|Activation|Deactivate (Variables|Default|GetNS_ExCalibur_Idle)))))

(event EventBeginPlay
  (|CallParentFunction))

(event EventTick
  (params (DeltaSeconds :type Float%20%28single-precision%29)))

(event EventOnPlayWeaponFireFX
  (params (bPersistentFire :type Boolean) (bSecondaryFire :type Boolean))
  (|CallParentFunction bPersistentFire bSecondaryFire)
  (Rendering|Material|SetScalarParameterValueonMaterials (Variables|Visual|GetWeaponMesh) "Opacity" 0.0)
  (Rendering|Material|SetScalarParameterValueonMaterials (Variables|Visual|GetWeaponMesh) "SwordReveal" 0.0)
  (Rendering|Material|SetScalarParameterValueonMaterials (Variables|Default|GetFXStaticMeshComponent) "Fresnel Brightness" 0.0)
  (Components|Activation|Deactivate (Variables|Default|GetNS_ExCalibur_Idle))
  (if (not (< (FortWeapon|GetMagazineAmmoCount) 2))
    (Utilities|FlowControl|Delay 0.9)
    (CallFunction|SwordRezIn)))

(event EventOnPlayReloadFX
  (params (ReloadStage :type EFortReloadFXState%20Enum :default "ReloadStart"))
  (|CallParentFunction ReloadStage))

(event EventMuzzlePlayReloadFX
  (params (Selection :type EFortReloadFXState%20Enum :default "ReloadStart"))
  (|CallParentFunction Selection)
  (switch Utilities|FlowControl|Switch|SwitchonEFortReloadFXState Selection
    (:ReloadStart
      (Utilities|FlowControl|Sequence
        (:then_0
          (bind (_direction _newtrack_1) (Timeline|Timeline_0 Stop)
            (:Update
              (Rendering|Material|SetVectorParameterValueonMaterials (Variables|Visual|GetWeaponMesh) "Emissive Color" (* (Variables|Default|GetEmissiveColorVector) _newtrack_1)))))
        (:then_1
          (bind (_direction_2 _newtrack_1_2) (Timeline|Timeline_3 PlayFromStart)
            (:Update
              (Rendering|Material|SetVectorParameterValueonMaterials (Variables|Visual|GetWeaponMesh) "Emissive Color" (* (Variables|Default|GetEmissiveColorVector) _newtrack_1_2)))))))
    (:ReloadCartridge
      (CallFunction|SwordRezIn))
    (:ReloadEnd
      (Utilities|FlowControl|Sequence
        (:then_0
          (Timeline|Timeline_3 Stop))
        (:then_1
          (bind (_direction_3 _newtrack_1_3) (Timeline|Timeline_2 PlayFromStart)
            (:Update
              (Rendering|Material|SetVectorParameterValueonMaterials (Variables|Visual|GetWeaponMesh) "Emissive Color" (* (Variables|Default|GetEmissiveColorVector) _newtrack_1_3)))))))))

(event Custom|SwordRezIn
  (Components|Activation|Activate (Variables|Default|GetSwordPhase) true)
  (Components|Activation|Activate (Variables|Default|GetNS_ExCalibur_Idle) true)
  (Rendering|Material|SetScalarParameterValueonMaterials (Variables|Visual|GetWeaponMesh) "Opacity" 1.0)
  (bind (_direction _newtrack_0) (Timeline|Timeline PlayFromStart)
    (:Update
      (Rendering|Material|SetScalarParameterValueonMaterials (Variables|Visual|GetWeaponMesh) "SwordReveal" _newtrack_0)
      (Rendering|Material|SetScalarParameterValueonMaterials (Variables|Default|GetFXStaticMeshComponent) "Fresnel Brightness" (* _newtrack_0 0.15)))))

(event EventMuzzleFlashFX
  (params (Persistent%20Fire :type Boolean))
  (|CallParentFunction Persistent%20Fire)
  (Timeline|Timeline_0 PlayFromStart))

(fn ConstructionScript
  (|CallParentFunction))
