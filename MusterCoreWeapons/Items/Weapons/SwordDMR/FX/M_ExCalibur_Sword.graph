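# Masked lit material for the ExCalibur sword of the SwordDMR weapon, with glowing sparkle detail.
# Emissive combines curvature, view-dependent glitter (off on mobile and low/medium quality) and thickness-masked noise projected on pre-skinned coordinates, with SwordReveal blending from solid color.
# A start time in custom primitive data triggers a decaying wobble that rotates position and normals around a pivot.
# A camera-centered sphere mask feeds temporal dither for proximity fade.
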Material /MusterCoreWeapons/Items/Weapons/SwordDMR/FX/M_ExCalibur_Sword
graph {
 properties {
  bUsedWithMeshDeformer=true
  bUsedWithSkeletalMesh=true
  blendMode=BLEND_Masked
  pixelDepthOffsetMode=PDOM_Legacy
  shadingModel=MSM_FromMaterialExpression
  usedShadingModels=Default Lit
 }
 nodes [
  Constant3Vector()
  LinearInterpolate(Alpha=NamedRerouteUsage())
  LinearInterpolate(Alpha=NamedRerouteUsage()) {constA=0.25}
  n1 = WorldPosition()
  n2 = CameraPositionWS()
  n3 = ScalarParameter() {defaultValue=300, parameterName=FadeMask}
  n4 = Constant()
  MaterialFunctionCall(Alpha Threshold=OneMinus(None=SphereMask(A=@n1.XYZ, B=@n2, Radius=@n3, Hardness=@n4))) {materialFunction=&DitherTemporalAA_Fortnite}
  n5 = Time() {period=1}
  n6 = Multiply(A=@n5, B=ScalarParameter() {defaultValue=2, parameterName=Wobble Frequency}) {constB=4}
  n7 = Normalize(VectorInput=MaterialFunctionCall(R=Sine(None=@n6), G=Cosine(None=@n6)) {materialFunction=&AppendMany}.RGB)
  n8 = ScalarParameter() {bUseCustomPrimitiveData=true, parameterName=WobblePlay}
  n9 = Saturate(None=Divide(A=Subtract(A=@n5, B=@n8), B=ScalarParameter() {defaultValue=0.5, parameterName=Wobble Duration}))
  n10 = Multiply(A=ScalarParameter() {defaultValue=0.03, parameterName=Wobble Amp}, B=Power(Base=OneMinus(None=@n9)) {constExponent=1.5})
  MaterialFunctionCall(
   BaseNormal=TextureSample() {samplerType=SAMPLERTYPE_Normal, texture=/MusterCoreWeapons/Items/Weapons/SwordDMR/Textures/T_Rifle_Ex_Caliber_sword_N}.RGB
   AdditionalNormal=MaterialFunctionCall(Rotation Axis=@n7, Rotation Angle=@n10) {materialFunction=&FixRotateAboutAxisNormals}.New Tangent Space Vertex Normal
  ) {materialFunction=&BlendAngleCorrectedNormals}
  n11 = VectorParameter() {group=Emissive, parameterName=EmissiveColor, sortPriority=0, defaultValue=<rgba 0.7499 0.1938 0.974 1>}
  n12 = VectorParameter() {group=Emissive, parameterName=EmissiveColor, sortPriority=0, defaultValue=<rgba 0.7499 0.1938 0.974 1>}
  n13 = TextureSample() {texture=&/T_FX_ExCalibur_Sword_Crv}
  n14 = ScalarParameter() {defaultValue=0.1, group=Emissive, parameterName=Curvature Emissive, sortPriority=2}
  n15 = ScalarParameter() {defaultValue=10, group=GlitterAnimated, parameterName=Sparkle Detail Tiling, sliderMax=64, sliderMin=8, sortPriority=2}
  n16 = Multiply(A=TextureCoordinate(), B=@n15) {constB=3}
  n17 = TextureSampleParameter2D(UVs=@n16) {
   group=GlitterAnimated
   parameterName=Sparkle Mask
   samplerSource=SSM_Wrap_WorldGroupSettings
   samplerType=SAMPLERTYPE_Masks
   sortPriority=0
   texture=/Game/Weapons/WeaponSkins/Textures/T_SparkleMask
  }
  n18 = Frac(None=@n16)
  n19 = MaterialFunctionCall(X=@n17.R, Y=@n17.R, Z=Divide(A=Subtract(A=@n16, B=@n18), B=@n15)) {desc=could be removed if no blending is needed, materialFunction=&MakeFloat3}
  n20 = Add(
   A=ComponentMask(None=@n19.Result) {b=true}
   B=DotProduct(
    A=MaterialFunctionCall() {materialFunction=&CameraDirectionVector}.Result
    B=PixelNormalWS()
   )
  )
  n21 = ScalarParameter() {defaultValue=24, group=GlitterAnimated, parameterName=Glitter Scale, sliderMax=2, sortPriority=3}
  n22 = Frac(None=Add(A=ComponentMask(None=@n19.Result) {g=true}, B=Multiply(A=@n20, B=@n21) {constB=0.5}))
  n23 = Multiply(A=Sine(None=MaterialFunctionCall(UV Channel=@n22) {materialFunction=&LinearGradient}.VGradient)) {constB=0.5}
  n24 = Multiply(
   A=ComponentMask(None=@n19.Result) {r=true}
   B=Power(
    Base=Add(A=@n23) {constB=0.5}
    Exp=ScalarParameter() {defaultValue=32, group=GlitterAnimated, parameterName=Fade Curve Exponent, sortPriority=4}
   ) {constExponent=32}
  )
  n25 = Saturate(None=Multiply(A=Power(Base=@n24) {constExponent=4}) {constB=4})
  n26 = ScalarParameter() {defaultValue=10, group=GlitterAnimated, parameterName=Sparkle Intensity, sliderMax=20, sortPriority=5}
  n27 = Constant()
  n28 = ShadingPathSwitch(Default=QualitySwitch(Default=Multiply(A=@n25, B=@n26), Low=@n27, Medium=@n27), Mobile=QualitySwitch(Default=@n27))
  n29 = Add(
   A=Add(A=Multiply(A=@n13.RGB, B=@n14), B=NamedRerouteDeclaration(None=@n28) {name=GlitterMask, nodeColor=<rgba 0 1 0.8125 1>})
   B=Multiply(A=NamedRerouteUsage(), B=ScalarParameter() {defaultValue=1, group=Emissive, parameterName=Noise Mult})
  )
  n30 = LinearInterpolate(
   A=@n11.RGB
   B=LinearInterpolate(A=Constant3Vector(), B=@n12.RGB, Alpha=Saturate(None=@n29))
   Alpha=ScalarParameter() {defaultValue=1, group=Emissive, parameterName=SwordReveal}
  )
  n31 = ScalarParameter() {defaultValue=1, group=Emissive, parameterName=HDR Mult, sortPriority=3}
  MaterialFunctionCall(Emissive=MaterialFunctionCall(LDR Input=@n30, HDR Tint=Multiply(A=@n30, B=@n31)) {materialFunction=&DeriveHDRfromLDR}.RemappedRange) {materialFunction=&MF_FX_EmissiveAdjustments}
  Multiply(
   A=TextureSample() {texture=/MusterCoreWeapons/Items/Weapons/SwordDMR/Textures/T_Rifle_Ex_Caliber_sword_D}.RGB
   B=ScalarParameter() {defaultValue=0.25, group=Emissive, parameterName=Base Color Brightness}
  ) {constB=0.25}
  n32 = TextureSampleParameter2D() {group=Thickness, parameterName=ThicknessMap, texture=/MusterCoreWeapons/Items/Weapons/SwordDMR/Textures/T_Rifle_Ex_Caliber_sword_T}
  n33 = MaterialFunctionCall(Float2=MaterialFunctionCall(
    TextureObject=TextureObject() {texture=/Game/Characters/Player/Male/Medium/Bodies/M_MED_Viper/Textures/Glitch/T_SoftNoise}
    TextureSize=ScalarParameter() {defaultValue=512, group=Detail Texture, parameterName=DetailTile_04}
    WorldPosition=NamedRerouteUsage()
    World Space Normal=NamedRerouteUsage()
   ) {materialFunction=/Game/Weapons/WeaponSkins/Wraps/MaterialFunctions/WorldAlignedTexture}.XYZ Texture) {materialFunction=&BreakOutFloat2Components}
  n34 = ScalarParameter() {defaultValue=0.75, group=Thickness, parameterName=Pow_02}
  n35 = Multiply(
   A=SphereMask(
    A=OneMinus(None=@n32.R)
    B=ScalarParameter() {defaultValue=1, group=Thickness, parameterName=ThicknessSample}
    Radius=ScalarParameter() {defaultValue=0.4, group=Thickness, parameterName=ThicknessSampleRadius}
    Hardness=ScalarParameter() {group=Thickness, parameterName=ThicknessSampleHardness}
   )
   B=LinearInterpolate(A=ScalarParameter() {defaultValue=0.25, group=Thickness, parameterName=NoiseThickness_Influence}, Alpha=Multiply(A=@n33.R, B=@n34))
  )
  n36 = NamedRerouteUsage()
  NamedRerouteDeclaration(None=Multiply(A=Clamp(None=@n35), B=LinearInterpolate(Alpha=@n36))) {name=EmissiveNoise, nodeColor=<rgba 0.8281 0 1 1>}
  n37 = PreSkinnedPosition()
  NamedRerouteDeclaration(None=ShaderStageSwitch(PixelShader=VertexInterpolator(VS=@n37).PS, VertexShader=@n37)) {name=PreskinnedPos, nodeColor=<rgba 0 1 0.4375 1>}
  n38 = MaterialFunctionCall(Float3=MaterialFunctionCall(
    TextureObject=TextureObject() {samplerType=SAMPLERTYPE_Masks, texture=/Game/Global/Textures/Default/Noise/T_Metal_Noise}
    TextureSize=ScalarParameter() {defaultValue=256, group=Detail Texture, parameterName=DetailTile_02}
    WorldPosition=NamedRerouteUsage()
    World Space Normal=NamedRerouteUsage()
   ) {materialFunction=&WorldAlignedTexture}.XYZ Texture) {materialFunction=&BreakOutFloat3Components}
  n39 = NamedRerouteDeclaration(None=Clamp(None=Power(Base=@n38.R) {constExponent=3})) {name=NoiseMaskSpec, nodeColor=<rgba 1 0.2812 0 1>}
  n40 = TextureCoordinate()
  n41 = Time()
  n42 = VectorParameter() {group=Emissive, parameterName=EmissiveNoisePanSpeed, defaultValue=<rgba 0 0.1 0 1>}
  n43 = AppendVector(A=@n42.R, B=@n42.R)
  n44 = MaterialFunctionCall(Float2=TextureSample(UVs=Panner(Coordinate=@n40, Time=@n41, Speed=@n43)) {texture=/Game/Global/Textures/Default/Noise/T_Noise_Cloud}.RGB) {materialFunction=&BreakOutFloat2Components}
  NamedRerouteDeclaration(None=Clamp(None=Multiply(A=@n39, B=@n44.R), Min=ScalarParameter() {group=Emissive, parameterName=MinGlow})) {name=PanningNoise, nodeColor=<rgba 0 0.6016 1 1>}
  n45 = PreSkinnedNormal()
  NamedRerouteDeclaration(None=ShaderStageSwitch(PixelShader=VertexInterpolator(VS=@n45).PS, VertexShader=@n45)) {name=PreSkinnedNormal, nodeColor=<rgba 0 1 0.6719 1>}
  n46 = TextureObject() {samplerType=SAMPLERTYPE_Masks, texture=&T_Metal_Noise}
  n47 = ScalarParameter() {defaultValue=768, group=Detail Texture, parameterName=DetailTile_03}
  n48 = NamedRerouteUsage()
  n49 = NamedRerouteUsage()
  n50 = Multiply(A=MaterialFunctionCall(Float3=MaterialFunctionCall(TextureObject=@n46, TextureSize=@n47, WorldPosition=@n48, World Space Normal=@n49) {materialFunction=&WorldAlignedTexture}.XYZ Texture) {materialFunction=&BreakOutFloat3Components}.B) {constB=1.5}
  NamedRerouteDeclaration(None=Clamp(None=Power(Base=@n50) {constExponent=12})) {name=NoiseMaskRough, nodeColor=<rgba 0 1 0.6953 1>}
  RotateAboutAxis(
   NormalizedRotationAxis=@n7
   RotationAngle=@n10
   PivotPoint=TransformPosition(None=Constant3Vector() {constant=<rgba 0 0 115 0>}) {transformType=TRANSFORMPOSSOURCE_World}
   Position=WorldPosition().XYZ
  )
  ScalarParameter() {defaultValue=1, group=Opacity, parameterName=Opacity, sortPriority=0}
 ]
}
