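# Niagara effect for the ExCalibur (SwordDMR) weapon, named for shells during reload, that draws a short-lived ribbon trail.
# A single Ribbon emitter spawns particles at a steady rate with fixed-width ribbons and randomized material values.
# FN_RibbonMaterial pans color and opacity along U, while a curve-driven ScaleColor fades alpha over each particle's brief life.
# User Overheat and SkeletalMesh parameters are exposed, but no shown module reads them.
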NiagaraSystem /MusterCoreWeapons/Items/Weapons/SwordDMR/FX/NS_ExCalibur_Shells_Reload
dependencies {usedDataInterfaces [&NiagaraDataInterfaceCurve], usedDynamicInputs [], usedModules [&SystemState]}
emitters {Ribbon {
  emitterUpdateScript [
   &EmitterState()
   &SpawnRate(SpawnRate=20, Use Spawn Probability=false)
  ]
  particleSpawnScript [&InitializeParticle(
    Lifetime=0.25
    Mass Max=2
    Mass Mode=::NewEnumerator3
    Ribbon Width=3.5
    Ribbon Width Mode=::NewEnumerator1
    Sprite Size Max=<3 14>
    Sprite Size Min=<2 6>
    Sprite Size Mode=::NewEnumerator2
    UsePositionOffset=false
    Write Material Random=true
   )]
  particleUpdateScript [
   &ParticleState()
   &FN_RibbonMaterial(Color Pan U Speed=-1, Opacity Pan U Speed=-1, Scale Alpha=0.5)
   &ScaleColor(
    Scale Alpha=&FloatFromCurve(FloatCurve=DataInterface() {propertyValues {Curve {keys [<at -0.0046 1.0113 RCIM_Cubic>,<at 1.0011 0.002 RCIM_Cubic>]}}}, Scale Curve=0.5)
    ScaleA=true
    ScaleRGB=true
   )
  ]
  rendererClasses [&NiagaraRibbonRendererProperties]
  renderers [{rendererClass=&NiagaraRibbonRendererProperties}]
 }}
summary {
 systemName=NS_ExCalibur_Shells_Reload
 emitters [{emitterName=Ribbon, rendererClasses [&NiagaraRibbonRendererProperties]}]
 userVariables [
  User.Overheat {type=NiagaraFloat, defaultValue=0, description=""}
  User.SkeletalMesh {
   type=NiagaraDataInterfaceSkeletalMesh
   defaultValue {
    struct=NiagaraExt_VariableValue_DataInterface
    value {dataInterface=&self:NiagaraDataInterfaceSkeletalMesh_0}
   }
  }
 ]
}
