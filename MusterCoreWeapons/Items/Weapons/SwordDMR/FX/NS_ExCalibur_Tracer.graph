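# Niagara tracer effect for the ExCalibur SwordDMR weapon, drawn as a short-lived ribbon beam.
# The DynamicBeam emitter bursts particles along a beam whose absolute end is the User.MuzzlePosition parameter.
# A curve over ribbon link order shapes the width, swelling near one end and tapering to both tips.
# Over each particle's brief life, curves shrink the width and fade alpha while the white color is brightened.
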NiagaraSystem /MusterCoreWeapons/Items/Weapons/SwordDMR/FX/NS_ExCalibur_Tracer
dependencies {usedDataInterfaces [&NiagaraDataInterfaceCurve], usedDynamicInputs [], usedModules [&SystemState]}
emitters {DynamicBeam {
  emitterUpdateScript [
   &EmitterState()
   /Niagara/Modules/Beams/BeamEmitterSetup(Absolute Beam End=true, Beam End:NiagaraPosition=$User.MuzzlePosition)
   &SpawnBurst_Instantaneous(Spawn Count=100, Use Spawn Probability=false, UseLoopCountLimit=false)
  ]
  particleSpawnScript [
   &InitializeParticle(Color Mode=::NewEnumerator0, Lifetime=0.2, Position Mode=::NewEnumerator0, Write Material Random=true)
   /Niagara/Modules/Beams/SpawnBeam()
   /Niagara/Modules/Beams/BeamWidth(Beam Width:NiagaraFloat=&FloatFromCurve(
     FloatCurve=DataInterface() {propertyValues {Curve {keys [
         {arriveTangent=9.2439, interpMode=RCIM_Cubic, leaveTangent=9.2439, tangentMode=RCTM_User, time=0, value=0}
         {arriveTangent=0.0019, interpMode=RCIM_Cubic, time=0.2516, value=0.9517}
         {arriveTangent=-1.5543, interpMode=RCIM_Cubic, leaveTangent=-1.5543, tangentMode=RCTM_User, time=1.0011, value=0.002}
        ]}}}
     CurveIndex=$Particles.RibbonLinkOrder
     Scale Curve=5
    ))
  ]
  particleUpdateScript [
   &ParticleState()
   /Niagara/Modules/Beams/UpdateBeam()
   /Niagara/Modules/Beams/ScaleBeamWidth(Beam Width Scale:NiagaraFloat=&FloatFromCurve(
     FloatCurve=DataInterface() {propertyValues {Curve {keys [{interpMode=RCIM_Cubic, tangentMode=RCTM_User, time=0, value=1}, {interpMode=RCIM_Cubic, tangentMode=RCTM_User, time=1, value=0}]}}}
     Scale Curve=2.5
    ))
   &Color(
    Color=&MultiplyLinearColor(Linear Color=<rgba 1 1 1 1>, Scale Factor=&MakeVector4_FromFloat(Float=2))
    Scale Alpha=&FloatFromCurve(FloatCurve=DataInterface() {propertyValues {Curve {keys [
         <at 0.1769 0.9943 RCIM_Cubic>
         {arriveTangent=-1.9828, interpMode=RCIM_Cubic, leaveTangent=-1.9828, tangentMode=RCTM_User, time=1.0011, value=0.002}
        ]}}})
   )
  ]
  rendererClasses [&NiagaraRibbonRendererProperties]
  renderers [{rendererClass=&NiagaraRibbonRendererProperties}]
 }}
summary {
 systemName=NS_ExCalibur_Tracer
 emitters [{emitterName=DynamicBeam, rendererClasses [&NiagaraRibbonRendererProperties]}]
 userVariables [User.MuzzlePosition {type=NiagaraPosition, description="", defaultValue {}}]
}
