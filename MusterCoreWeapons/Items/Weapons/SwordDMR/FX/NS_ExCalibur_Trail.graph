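# Niagara ribbon trail effect for the ExCalibur SwordDMR weapon, built as a single emitter.
# Custom ribbon modules spawn, initialize and update the strip, while FN_RibbonMaterial pans color, erosion and opacity.
# Material alpha multiplies one curve by another indexed on normalized particle count, and further curves scale width and fade alpha.
# Velocity and color-lerp modules are disabled, leaving the user Light/Dark colors unused in the visible stack.
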NiagaraSystem /MusterCoreWeapons/Items/Weapons/SwordDMR/FX/NS_ExCalibur_Trail
dependencies {
 usedDataInterfaces [&NiagaraDataInterfaceSkeletalMesh, &NiagaraDataInterfaceCurve]
 usedDynamicInputs []
 usedModules [&SystemState]
}
emitters {Ribbon_Trail {
  emitterUpdateScript [
   &EmitterState(Life Cycle Mode=::NewEnumerator1, Loop Duration=1, Scalability Mode=::NewEnumerator1, UseLoopDelay=false)
   &SpawnBurst_Instantaneous(Spawn Count=1, Use Spawn Probability=false)
   &FN_SpawnRibbon_v2(Spawn Mode=::NewEnumerator0, SpawnRate=60)
  ]
  particleSpawnScript [
   /Game/Effects/Niagara/Modules/Ribbons/FN_InitializeRibbon_v4(Offset Amount=15, Offset Axis:ENiagaraOrientationAxis=::XAxis, Ribbon Width=50)
   &AddVelocity(Velocity=<-100 0 0>) {enabled=false}
  ]
  particleUpdateScript [
   &SolveForcesAndVelocity(Clamp Velocity=false, Limit Acceleration=false) {enabled=false}
   /Game/Effects/Niagara/Modules/Ribbons/FN_UpdateRibbon_v6()
   &FN_RibbonMaterial(
    Color Pan U Speed=1
    Erosion Pan Speed U=1
    Opacity Pan U Speed=1
    Scale Alpha:NiagaraFloat=&Multiply_Float(
     A=&FloatFromCurve(
      FloatCurve=DataInterface() {propertyValues {Curve {keys [
          {arriveTangent=-5.8323, interpMode=RCIM_Cubic, leaveTangent=-5.8323, tangentMode=RCTM_User, time=0, value=1}
          <at 0.2 0.4>
          {arriveTangent=-1, interpMode=RCIM_Cubic, leaveTangent=-1, time=0.5, value=0.4}
          <at 1 0 RCIM_Cubic>
         ]}}}
      Scale Curve=2
     )
     B=&FloatFromCurve(
      FloatCurve=DataInterface() {propertyValues {Curve {keys [{arriveTangent=-0.0002, interpMode=RCIM_Cubic, leaveTangent=-0.0002, tangentMode=RCTM_User, time=0.8, value=1}, <at 1 0 RCIM_Cubic>]}}}
      CurveIndex=/Game/Effects/Niagara/DynamicInputs/Math/FN_NormalizedParticleCount()
     )
    )
   )
   &Color(Color=&LerpLinearColors(StartColor=$User.LightColor, EndColor=$User.DarkColor, LerpFactor=$Particles.NormalizedAge)) {enabled=false}
   &ScaleRibbonWidth(Ribbon Width Scale=&FloatFromCurve(FloatCurve=DataInterface() {propertyValues {Curve {keys [
         {arriveTangent=0.6919, interpMode=RCIM_Cubic, time=0.5, value=0.5}
         {arriveTangent=0.0237, interpMode=RCIM_Cubic, leaveTangent=0.0237, tangentMode=RCTM_User, time=1, value=1}
        ]}}}))
   &ScaleColor(
    Scale Alpha=&FloatFromCurve(FloatCurve=DataInterface() {propertyValues {Curve {keys [{interpMode=RCIM_Cubic, tangentMode=RCTM_User, time=0, value=1}, {tangentMode=RCTM_User, time=1, value=0}]}}})
    Scale RGB=&VectorFromFloat(Value=1.8)
    ScaleA=true
    ScaleRGB=true
   )
  ]
  rendererClasses [&NiagaraRibbonRendererProperties]
  renderers [{rendererClass=&NiagaraRibbonRendererProperties}]
 }}
summary {
 systemName=NS_ExCalibur_Trail
 emitters [{emitterName=Ribbon_Trail, rendererClasses [&NiagaraRibbonRendererProperties]}]
 userVariables [
  User.DarkColor {type=LinearColor, description="", defaultValue {struct=LinearColor, value=<rgba 0.1968 0 0.974 1>}}
  User.LightColor {type=LinearColor, description="", defaultValue {struct=LinearColor, value=<rgba 0.4876 0.3145 0.974 1>}}
  User.RibbonIDOffset {type=NiagaraInt32}
 ]
}
