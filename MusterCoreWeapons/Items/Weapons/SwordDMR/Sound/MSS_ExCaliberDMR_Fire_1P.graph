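# First-person fire MetaSound source for the Ex-Caliber DMR, output as a 5.1 one-shot.
# It wraps one shared MSS_Base_Weapon_Fire_1p_Complete node, passing every input through and returning its audio, gunshot and finish outputs.
# This asset supplies main fire, mechanism, low-ammo and sniper suppressor waves plus mix and randomization defaults; sweetener and stop-fire arrays are empty.
# Playback uses the local gunfire concurrency set.
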MetaSoundSource /MusterCoreWeapons/Items/Weapons/SwordDMR/Sound/MSS_ExCaliberDMR_Fire_1P
inputs [
 bIsAlwaysSuppressed {type=Bool}
 bIsFiring {type=Bool}
 BurstRate {type=Time}
 bUseSingleVoice {type=Bool}
 FireBurst {type=Trigger}
 FireSingleShot {type=Trigger}
 LowAmmo.ADS.VolumeMultiply {type=Float}
 LowAmmo.Random.Pitch.Max {type=Float}
 LowAmmo.Random.Pitch.Min {type=Float}
 LowAmmo.Random.Volume.Max {type=Float}
 LowAmmo.Random.Volume.Min {type=Float}
 LowAmmo.Sounds {type=WaveAsset:Array}
 LowAmmo.Threshold {type=Int32}
 LowAmmo.Volume {type=Float}
 MainFire.ADS.VolumeMultiply {type=Float}
 MainFire.Random.Pitch.Max {type=Float}
 MainFire.Random.Pitch.Min {type=Float}
 MainFire.Random.Volume.Max {type=Float}
 MainFire.Random.Volume.Min {type=Float}
 MainFire.Sounds {type=WaveAsset:Array}
 MainFire.Volume {type=Float}
 MasterVolume {type=Float}
 Mech.ADS.VolumeMultiply {type=Float}
 Mech.Random.Pitch.Max {type=Float}
 Mech.Random.Pitch.Min {type=Float}
 Mech.Random.Volume.Max {type=Float}
 Mech.Random.Volume.Min {type=Float}
 Mech.Sounds {type=WaveAsset:Array}
 Mech.Volume {type=Float}
 OnStopFire.Sounds {type=WaveAsset:Array}
 OnStopFire.Volume {type=Float}
 RetriggerDelayTime {type=Time}
 RoundsPerBurst {type=Int32}
 StopBurst {type=Trigger}
 Suppressor.BaseWeaponMix {type=Float}
 Suppressor.EnvelopeDecayTime {type=Time}
 Suppressor.MasterVolume {type=Float}
 Suppressor.Random.Pitch.Max {type=Float}
 Suppressor.Random.Pitch.Min {type=Float}
 Suppressor.Random.Volume.Max {type=Float}
 Suppressor.Random.Volume.Min {type=Float}
 Suppressor.Sounds {type=WaveAsset:Array}
 Suppressor.SuppressorVolume {type=Float}
 Sweetener.ADS.VolumeMultiply {type=Float}
 Sweetener.Random.Pitch.Max {type=Float}
 Sweetener.Random.Pitch.Min {type=Float}
 Sweetener.Random.Volume.Max {type=Float}
 Sweetener.Random.Volume.Min {type=Float}
 Sweetener.Sounds {type=WaveAsset:Array}
 Sweetener.Volume {type=Float}
 UE.Source.OnPlay {type=Trigger}
 Weapon.AmmoRemaining {type=Int32}
 Weapon.ClipSize {type=Int32}
 Weapon.FireRate {type=Float}
 Weapon.FireStartTime {type=Float}
 Weapon.IsAimingDownSights {type=Bool}
 Weapon.IsSuppressorEnabled {type=Bool}
]
interfaces [Fort.Weapon:1.0, UE.OutputFormat.5dot1:1.0, UE.Source.OneShot:1.0, UE.Source:1.1]
outputs [
 bIsAlmostFinished {type=Bool}
 OnGunshotOut {type=Trigger}
 "UE.OutputFormat.5dot1.Audio:0" {type=Audio}
 "UE.OutputFormat.5dot1.Audio:1" {type=Audio}
 "UE.OutputFormat.5dot1.Audio:2" {type=Audio}
 "UE.OutputFormat.5dot1.Audio:3" {type=Audio}
 "UE.OutputFormat.5dot1.Audio:4" {type=Audio}
 "UE.OutputFormat.5dot1.Audio:5" {type=Audio}
 UE.Source.OneShot.OnFinished {type=Trigger}
]
pages [Default {
  input_defaults {
   BurstRate=0.4
   FireBurst=null
   FireSingleShot=null
   LowAmmo.ADS.VolumeMultiply=1
   LowAmmo.Random.Pitch.Max=1.05
   LowAmmo.Random.Pitch.Min=0.95
   LowAmmo.Random.Volume.Max=1.05
   LowAmmo.Random.Volume.Min=0.95
   LowAmmo.Threshold=1
   LowAmmo.Volume=1
   MainFire.ADS.VolumeMultiply=1
   MainFire.Random.Pitch.Max=1.05
   MainFire.Random.Pitch.Min=0.95
   MainFire.Random.Volume.Max=1.05
   MainFire.Random.Volume.Min=0.95
   MainFire.Volume=1
   MasterVolume=1
   Mech.ADS.VolumeMultiply=1
   Mech.Random.Pitch.Max=1.05
   Mech.Random.Pitch.Min=0.95
   Mech.Random.Volume.Max=1.05
   Mech.Random.Volume.Min=0.95
   Mech.Volume=1
   OnStopFire.Volume=1
   RetriggerDelayTime=2
   RoundsPerBurst=2
   StopBurst=null
   Suppressor.BaseWeaponMix=0.9
   Suppressor.EnvelopeDecayTime=0.195
   Suppressor.MasterVolume=1
   Suppressor.Random.Pitch.Max=1.05
   Suppressor.Random.Pitch.Min=0.95
   Suppressor.Random.Volume.Max=1.05
   Suppressor.Random.Volume.Min=0.95
   Suppressor.SuppressorVolume=1
   Sweetener.ADS.VolumeMultiply=1
   Sweetener.Random.Pitch.Max=1.05
   Sweetener.Random.Pitch.Min=0.95
   Sweetener.Random.Volume.Max=1.05
   Sweetener.Random.Volume.Min=0.95
   Sweetener.Volume=1
   UE.Source.OnPlay=null
   Weapon.AmmoRemaining=30
   Weapon.ClipSize=30
   Weapon.FireRate=0.5
   Weapon.FireStartTime=0
   Weapon.IsAimingDownSights=null
   Weapon.IsSuppressorEnabled=null
   bIsAlwaysSuppressed=null
   bIsFiring=null
   bUseSingleVoice="<MetaSoundBuilderResult.SUCCEEDED: 0>"
   LowAmmo.Sounds [&/Ex-CaliberDMR_LowAmmo]
   MainFire.Sounds [
    &/EXCaliber_Sniper_Fire_1P_01
    &/EXCaliber_Sniper_Fire_1P_02
    &/EXCaliber_Sniper_Fire_1P_03
    &/EXCaliber_Sniper_Fire_1P_04
    &/EXCaliber_Sniper_Fire_1P_05
    &/EXCaliber_Sniper_Fire_1P_06
   ]
   Mech.Sounds [
    &/ExcaliberDMR_Mech_Fire_06
    &/ExcaliberDMR_Mech_Fire_05
    &/ExcaliberDMR_Mech_Fire_04
    &/ExcaliberDMR_Mech_Fire_03
    &/ExcaliberDMR_Mech_Fire_02
    &/ExcaliberDMR_Mech_Fire_01
   ]
   OnStopFire.Sounds []
   Suppressor.Sounds [
    /Game/Sounds/Fort_Weapons/Suppressor/Silencer_Sniper_Close_01
    /Game/Sounds/Fort_Weapons/Suppressor/Silencer_Sniper_Close_02
    /Game/Sounds/Fort_Weapons/Suppressor/Silencer_Sniper_Close_03
    /Game/Sounds/Fort_Weapons/Suppressor/Silencer_Sniper_Close_04
    /Game/Sounds/Fort_Weapons/Suppressor/Silencer_Sniper_Close_05
   ]
   Sweetener.Sounds []
  }
  nodes [MSS_Base_Weapon_Fire_1p_Complete_0 = /FortMetaSoundWeapon/Items/Metasounds/MSS_Base_Weapon_Fire_1p_Complete(
    BurstRate=^BurstRate
    FireBurst=^FireBurst
    FireSingleShot=^FireSingleShot
    LowAmmo.ADS.VolumeMultiply=^LowAmmo.ADS.VolumeMultiply
    LowAmmo.Random.Pitch.Max=^LowAmmo.Random.Pitch.Max
    LowAmmo.Random.Pitch.Min=^LowAmmo.Random.Pitch.Min
    LowAmmo.Random.Volume.Max=^LowAmmo.Random.Volume.Max
    LowAmmo.Random.Volume.Min=^LowAmmo.Random.Volume.Min
    LowAmmo.Sounds=^LowAmmo.Sounds
    LowAmmo.Threshold=^LowAmmo.Threshold
    LowAmmo.Volume=^LowAmmo.Volume
    MainFire.ADS.VolumeMultiply=^MainFire.ADS.VolumeMultiply
    MainFire.Random.Pitch.Max=^MainFire.Random.Pitch.Max
    MainFire.Random.Pitch.Min=^MainFire.Random.Pitch.Min
    MainFire.Random.Volume.Max=^MainFire.Random.Volume.Max
    MainFire.Random.Volume.Min=^MainFire.Random.Volume.Min
    MainFire.Sounds=^MainFire.Sounds
    MainFire.Volume=^MainFire.Volume
    MasterVolume=^MasterVolume
    Mech.ADS.VolumeMultiply=^Mech.ADS.VolumeMultiply
    Mech.Random.Pitch.Max=^Mech.Random.Pitch.Max
    Mech.Random.Pitch.Min=^Mech.Random.Pitch.Min
    Mech.Random.Volume.Max=^Mech.Random.Volume.Max
    Mech.Random.Volume.Min=^Mech.Random.Volume.Min
    Mech.Sounds=^Mech.Sounds
    Mech.Volume=^Mech.Volume
    OnStopFire.Sounds=^OnStopFire.Sounds
    OnStopFire.Volume=^OnStopFire.Volume
    RetriggerDelayTime=^RetriggerDelayTime
    RoundsPerBurst=^RoundsPerBurst
    StopBurst=^StopBurst
    Suppressor.BaseWeaponMix=^Suppressor.BaseWeaponMix
    Suppressor.EnvelopeDecayTime=^Suppressor.EnvelopeDecayTime
    Suppressor.MasterVolume=^Suppressor.MasterVolume
    Suppressor.Random.Pitch.Max=^Suppressor.Random.Pitch.Max
    Suppressor.Random.Pitch.Min=^Suppressor.Random.Pitch.Min
    Suppressor.Random.Volume.Max=^Suppressor.Random.Volume.Max
    Suppressor.Random.Volume.Min=^Suppressor.Random.Volume.Min
    Suppressor.Sounds=^Suppressor.Sounds
    Suppressor.SuppressorVolume=^Suppressor.SuppressorVolume
    Sweetener.ADS.VolumeMultiply=^Sweetener.ADS.VolumeMultiply
    Sweetener.Random.Pitch.Max=^Sweetener.Random.Pitch.Max
    Sweetener.Random.Pitch.Min=^Sweetener.Random.Pitch.Min
    Sweetener.Random.Volume.Max=^Sweetener.Random.Volume.Max
    Sweetener.Random.Volume.Min=^Sweetener.Random.Volume.Min
    Sweetener.Sounds=^Sweetener.Sounds
    Sweetener.Volume=^Sweetener.Volume
    UE.Source.OnPlay=^UE.Source.OnPlay
    Weapon.AmmoRemaining=^Weapon.AmmoRemaining
    Weapon.ClipSize=^Weapon.ClipSize
    Weapon.FireRate=^Weapon.FireRate
    Weapon.FireStartTime=^Weapon.FireStartTime
    Weapon.IsAimingDownSights=^Weapon.IsAimingDownSights
    Weapon.IsSuppressorEnabled=^Weapon.IsSuppressorEnabled
    bIsAlwaysSuppressed=^bIsAlwaysSuppressed
    bIsFiring=^bIsFiring
    bUseSingleVoice=^bUseSingleVoice
   )]
  links [
   {dst="@OnGunshotOut", src=MSS_Base_Weapon_Fire_1p_Complete_0.OnGunshotOut}
   {dst="@UE.OutputFormat.5dot1.Audio:0", src=MSS_Base_Weapon_Fire_1p_Complete_0.UE.OutputFormat.5dot1.Audio:0}
   {dst="@UE.OutputFormat.5dot1.Audio:1", src=MSS_Base_Weapon_Fire_1p_Complete_0.UE.OutputFormat.5dot1.Audio:1}
   {dst="@UE.OutputFormat.5dot1.Audio:2", src=MSS_Base_Weapon_Fire_1p_Complete_0.UE.OutputFormat.5dot1.Audio:2}
   {dst="@UE.OutputFormat.5dot1.Audio:3", src=MSS_Base_Weapon_Fire_1p_Complete_0.UE.OutputFormat.5dot1.Audio:3}
   {dst="@UE.OutputFormat.5dot1.Audio:4", src=MSS_Base_Weapon_Fire_1p_Complete_0.UE.OutputFormat.5dot1.Audio:4}
   {dst="@UE.OutputFormat.5dot1.Audio:5", src=MSS_Base_Weapon_Fire_1p_Complete_0.UE.OutputFormat.5dot1.Audio:5}
   {dst="@UE.Source.OneShot.OnFinished", src=MSS_Base_Weapon_Fire_1p_Complete_0.UE.Source.OneShot.OnFinished}
   {dst="@bIsAlmostFinished", src=MSS_Base_Weapon_Fire_1p_Complete_0.bIsAlmostFinished}
  ]
 }]
properties {
 assetImportData=&self:AssetImportData
 numChannels=6
 outputFormat=FiveDotOne
 qualitySetting=Fort_Weapons
 sampleRate=48000
 soundClassObject=&Fort_WeaponRanged_Fire_1P_PSM
 volume=0.9
 concurrencyOverrides {maxCount=2, resolutionRule=StopOldest}
 concurrencySet [&Weapon_Gunfire_Local_Conc]
}
