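# Sound cue for distant bullet whip sounds of the ExCaliber DMR weapon, classed as a ranged projectile effect.
# A quality-level switch feeds two tiers a random pick from eight distant whip waves and the other two a reduced pool of three.
# A modulator at the root varies pitch slightly downward.
# The cue uses shared bullet-whip attenuation and whiz-by concurrency assets, with boosted volume.
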SoundCue /MusterCoreWeapons/Items/Weapons/SwordDMR/Sound/SC_ExCaliberDMR_BulletWhip_Far
root=Modulator_1
nodes [
 {id=Modulator_1, children [QualityLevel_1], properties {pitchMax=1, pitchMin=0.9}}
 {id=QualityLevel_1, children [Random_2,Random_2,Random_3,Random_3]}
 {
  id=Random_2
  children [WavePlayer_5, WavePlayer_7, WavePlayer_8, WavePlayer_9, WavePlayer_10, WavePlayer_11, WavePlayer_12, WavePlayer_13]
  properties {weights [1,1,1,1,1,1,1,1]}
 }
 {id=Random_3, children [WavePlayer_11, WavePlayer_12, WavePlayer_13], properties {bShouldExcludeFromBranchCulling=true, weights [1,1,1]}}
 {id=WavePlayer_10, properties {soundWaveAssetPtr=&/ExCaliberDMR_Bullet_Whip_Distant_05}}
 {id=WavePlayer_11, properties {soundWaveAssetPtr=&/ExCaliberDMR_Bullet_Whip_Distant_06}}
 {id=WavePlayer_12, properties {soundWaveAssetPtr=&/ExCaliberDMR_Bullet_Whip_Distant_07}}
 {id=WavePlayer_13, properties {soundWaveAssetPtr=&/ExCaliberDMR_Bullet_Whip_Distant_08}}
 {id=WavePlayer_5, properties {soundWaveAssetPtr=&/ExCaliberDMR_Bullet_Whip_Distant_01}}
 {id=WavePlayer_7, properties {soundWaveAssetPtr=&/ExCaliberDMR_Bullet_Whip_Distant_02}}
 {id=WavePlayer_8, properties {soundWaveAssetPtr=&/ExCaliberDMR_Bullet_Whip_Distant_03}}
 {id=WavePlayer_9, properties {soundWaveAssetPtr=&/ExCaliberDMR_Bullet_Whip_Distant_04}}
]
properties {
 attenuationSettings=/Game/Athena/Sounds/Weapons/BulletWhips/bullet_whip_Att.Bullet_Whip_Att
 maxDistance=2097152
 soundClassObject=&Fort_WeaponRanged_Projectile_Effects
 volumeMultiplier=1.45
 attenuationOverrides {bAttenuate=false, stereoSpread=0}
 concurrencySet [/Game/Sounds/Concurrency_Rules/Athena_BulletWhizBy_Conc]
}
