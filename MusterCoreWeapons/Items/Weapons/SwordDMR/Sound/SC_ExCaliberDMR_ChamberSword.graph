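# Weapon foley SoundCue for chambering the sword in the Ex-Caliber DMR, using weapon reload attenuation.
# A mixer layers a random chamber-sword sample with a quieter, briefly delayed random reload-zap sample.
# Each layer passes through a quality-level switch, drawing on four variants in the first two slots and only two in the last two.
# The mix is routed to the weapon foley sound class, with slight volume randomization on top.
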SoundCue /MusterCoreWeapons/Items/Weapons/SwordDMR/Sound/SC_ExCaliberDMR_ChamberSword
root=Attenuation_0
nodes [
 {id=Attenuation_0, children [Modulator_0], properties {attenuationSettings=&Fort_Weapon_Reload_Att}}
 {id=Delay_0, children [QualityLevel_0], properties {delayMax=0.2, delayMin=0.2}}
 {id=Mixer_0, children [QualityLevel_1, Modulator_1], properties {inputVolume [1,0.3]}}
 {id=Modulator_0, children [SoundClass_0], properties {volumeMax=0.6, volumeMin=0.5}}
 {id=Modulator_1, children [Delay_0], properties {pitchMax=1, pitchMin=1, volumeMax=1, volumeMin=1}}
 {id=QualityLevel_0, children [Random_1,Random_1,Random_2,Random_2]}
 {id=QualityLevel_1, children [Random_0,Random_0,Random_3,Random_3]}
 {id=Random_0, children [WavePlayer_1,WavePlayer_2,WavePlayer_3,WavePlayer_4], properties {weights [1,1,1,1]}}
 {id=Random_1, children [WavePlayer_0,WavePlayer_5,WavePlayer_6,WavePlayer_7], properties {weights [1,1,1,1]}}
 {id=Random_2, children [WavePlayer_6,WavePlayer_7], properties {weights [1,1]}}
 {id=Random_3, children [WavePlayer_3,WavePlayer_4], properties {weights [1,1]}}
 {id=SoundClass_0, children [Mixer_0], properties {soundClassOverride=&Fort_WeaponFoley}}
 {id=WavePlayer_0, properties {soundWaveAssetPtr=&/ExCaliberDMR_ReloadZaps_01}}
 {id=WavePlayer_1, properties {soundWaveAssetPtr=&/Ex-CaliberDMR_ChamberSword_01}}
 {id=WavePlayer_2, properties {soundWaveAssetPtr=&/Ex-CaliberDMR_ChamberSword_02}}
 {id=WavePlayer_3, properties {soundWaveAssetPtr=&/Ex-CaliberDMR_ChamberSword_03}}
 {id=WavePlayer_4, properties {soundWaveAssetPtr=&/Ex-CaliberDMR_ChamberSword_04}}
 {id=WavePlayer_5, properties {soundWaveAssetPtr=&/ExCaliberDMR_ReloadZaps_02}}
 {id=WavePlayer_6, properties {soundWaveAssetPtr=&/ExCaliberDMR_ReloadZaps_03}}
 {id=WavePlayer_7, properties {soundWaveAssetPtr=&/ExCaliberDMR_ReloadZaps_04}}
]
properties {
 attenuationSettings=&Fort_Weapon_Reload_Att
 bDebug=true
 bOverrideConcurrency=true
 maxDistance=2250
 soundClassObject=&Fort_WeaponFoley
 volumeMultiplier=1.504
 attenuationOverrides {falloffDistance=1500, attenuationShapeExtents {x=500}}
}
