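# Sound Cue for the third close-chamber foley sound of the Ex-Caliber DMR weapon.
# A single wave player feeds a SoundClass node set to weapon foley, then a Modulator that slightly randomizes volume at fixed pitch, offset by a boosted cue volume multiplier.
# Inline attenuation overrides replace the assigned reload attenuation asset, concurrency is also overridden, and the debug flag is left on.
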SoundCue /MusterCoreWeapons/Items/Weapons/SwordDMR/Sound/SC_ExCaliberDMR_CloseChamber3
root=Modulator_0
nodes [
 {id=Modulator_0, children [SoundClass_0], properties {pitchMax=1, pitchMin=1, volumeMax=0.6, volumeMin=0.5}}
 {id=SoundClass_0, children [WavePlayer_1], properties {soundClassOverride=&Fort_WeaponFoley}}
 {id=WavePlayer_1, properties {soundWaveAssetPtr=&/Ex-CaliberDMR_CloseChamber3A}}
]
properties {
 attenuationSettings=&Fort_Weapon_Reload_Att
 bDebug=true
 bOverrideAttenuation=true
 bOverrideConcurrency=true
 maxDistance=2000
 soundClassObject=&Fort_WeaponFoley
 volumeMultiplier=2.506
 attenuationOverrides {falloffDistance=1500, stereoSpread=300, attenuationShapeExtents {x=500}}
}
