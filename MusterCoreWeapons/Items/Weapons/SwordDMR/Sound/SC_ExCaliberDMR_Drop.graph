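# Sound cue for the ExCaliber DMR sword weapon's drop sound.
# A single drop wave plays through a Modulator at default settings, adding slight random pitch and volume variation.
# The cue plays at reduced volume in the loot-collect sound class, with shared weapon-drop attenuation.
# The shared Weapon_Drop_StopOldest concurrency set limits overlapping drops.
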SoundCue /MusterCoreWeapons/Items/Weapons/SwordDMR/Sound/SC_ExCaliberDMR_Drop
root=Modulator_0
nodes [
 {id=Modulator_0, children [WavePlayer_1]}
 {id=WavePlayer_1, properties {soundWaveAssetPtr=&/ExCaliberDMR_DropNew}}
]
properties {
 attenuationSettings=&Fort_Weapon_Drop_Att
 maxDistance=3000
 soundClassObject=&Fort_Loot_Collect
 volumeMultiplier=0.6
 concurrencyOverrides {resolutionRule=PreventNew}
 concurrencySet [&Weapon_Drop_StopOldest]
}
