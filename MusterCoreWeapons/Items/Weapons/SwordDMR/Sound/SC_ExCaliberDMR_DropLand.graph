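# Sound cue that plays when the ExCaliber DMR weapon drops and lands as loot in the world.
# A mixer layers the drop-land wave with a reload zap wave, with the zap layer passing through a fixed modulator, and an outer modulator then varies the combined output.
# It plays at reduced volume using weapon drop attenuation, the loot collect sound class and a weapon drop concurrency set.
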SoundCue /MusterCoreWeapons/Items/Weapons/SwordDMR/Sound/SC_ExCaliberDMR_DropLand
root=Modulator_0
nodes [
 {id=Mixer_0, children [WavePlayer_0,Modulator_1], properties {inputVolume [1,1]}}
 {id=Modulator_0, children [Mixer_0]}
 {id=Modulator_1, children [WavePlayer_2], properties {pitchMax=1, pitchMin=1, volumeMax=1, volumeMin=1}}
 {id=WavePlayer_0, properties {soundWaveAssetPtr=&/ExCaliberDMR_DropLand}}
 {id=WavePlayer_2, properties {soundWaveAssetPtr=&/ExCaliberDMR_ReloadZaps_01}}
]
properties {
 attenuationSettings=&Fort_Weapon_Drop_Att
 maxDistance=3000
 soundClassObject=&Fort_Loot_Collect
 volumeMultiplier=0.3
 concurrencyOverrides {resolutionRule=PreventNew}
 concurrencySet [&Weapon_Drop_StopOldest]
}
