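# First-person fire SoundCue for the Ex-Caliber DMR weapon, non-spatialized and using first-person gunfire concurrency.
# A mixer layers a random fire shot with a half-volume shoot tail that is briefly delayed and pitched down.
# Quality-level nodes give each layer either six or three random wave variants.
# A root SoundClass node overrides to the 3P fire class, although the cue's own class is 1P.
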SoundCue /MusterCoreWeapons/Items/Weapons/SwordDMR/Sound/SC_ExCaliberDMR_Fire_1P
root=SoundClass_3
nodes [
 {id=Delay_0, children [Modulator_1], properties {delayMax=0.2, delayMin=0.2}}
 {id=Mixer_0, children [Modulator_3,Delay_0], properties {inputVolume [1,0.5]}}
 {id=Modulator_1, children [QualityLevel_1], properties {pitchMax=0.8, pitchMin=0.8, volumeMax=1, volumeMin=1}}
 {id=Modulator_2, children [QualityLevel_0]}
 {id=Modulator_3, children [Modulator_2], properties {pitchMax=1, pitchMin=0.98, volumeMax=1.2, volumeMin=1.2}}
 {id=QualityLevel_0, children [Random_0,Random_0,Random_1,Random_1]}
 {id=QualityLevel_1, children [Random_2,Random_2,Random_3,Random_3]}
 {
  id=Random_0
  children [WavePlayer_0, WavePlayer_1, WavePlayer_3, WavePlayer_4, WavePlayer_5, WavePlayer_16]
  properties {weights [1,1,1,1,1,1]}
 }
 {id=Random_1, children [WavePlayer_4, WavePlayer_5, WavePlayer_16], properties {weights [1,1,1]}}
 {
  id=Random_2
  children [WavePlayer_2, WavePlayer_6, WavePlayer_7, WavePlayer_8, WavePlayer_9, WavePlayer_10]
  properties {weights [1,1,1,1,1,1]}
 }
 {id=Random_3, children [WavePlayer_8, WavePlayer_9, WavePlayer_10], properties {weights [1,1,1]}}
 {id=SoundClass_3, children [Mixer_0], properties {soundClassOverride=&Fort_WeaponRanged_Fire_3P_PSM}}
 {id=WavePlayer_0, properties {soundWaveAssetPtr=&/Ex-CaliberDMR_Fire_1P_01}}
 {id=WavePlayer_1, properties {soundWaveAssetPtr=&/Ex-CaliberDMR_Fire_1P_02}}
 {id=WavePlayer_10, properties {soundWaveAssetPtr=&/Ex-CaliberDMR_ShootTail_06}}
 {id=WavePlayer_16, properties {soundWaveAssetPtr=&/Ex-CaliberDMR_Fire_1P_06}}
 {id=WavePlayer_2, properties {soundWaveAssetPtr=&/Ex-CaliberDMR_ShootTail_01}}
 {id=WavePlayer_3, properties {soundWaveAssetPtr=&/Ex-CaliberDMR_Fire_1P_03}}
 {id=WavePlayer_4, properties {soundWaveAssetPtr=&/Ex-CaliberDMR_Fire_1P_04}}
 {id=WavePlayer_5, properties {soundWaveAssetPtr=&/Ex-CaliberDMR_Fire_1P_05}}
 {id=WavePlayer_6, properties {soundWaveAssetPtr=&/Ex-CaliberDMR_ShootTail_02}}
 {id=WavePlayer_7, properties {soundWaveAssetPtr=&/Ex-CaliberDMR_ShootTail_03}}
 {id=WavePlayer_8, properties {soundWaveAssetPtr=&/Ex-CaliberDMR_ShootTail_04}}
 {id=WavePlayer_9, properties {soundWaveAssetPtr=&/Ex-CaliberDMR_ShootTail_05}}
]
properties {
 attenuationSettings=&Fort_Weapon_Projectile_FirstPerson_Athena_Att
 bDebug=true
 maxDistance=2097152
 soundClassObject=&Fort_WeaponRanged_Fire_1P_PSM
 volumeMultiplier=0.8
 attenuationOverrides {bSpatialize=false}
 concurrencyOverrides {bLimitToOwner=true, maxCount=3, resolutionRule=StopOldest}
 concurrencySet [&Weapon_Gunfire_FirstPerson_Conc]
}
