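# Third-person gunfire sound cue for the ExCaliber (Sword) DMR, built from close, medium and distant layers.
# Each layer has its own attenuation, volume modulation and sound class override, with pitch variation on close and medium, and all three are mixed together.
# Quality-level nodes switch each layer between a full random wave pool and a smaller subset of the same waves.
# Global and local gunfire concurrency sets are assigned.
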SoundCue /MusterCoreWeapons/Items/Weapons/SwordDMR/Sound/SC_ExCaliberDMR_Fire_3P
root=Mixer_2
nodes [
 {id=Attenuation_3, children [Modulator_0], properties {attenuationSettings=&Fort_Weapon_Projectile_Close_Athena_Att}}
 {
  id=Attenuation_4
  children [Modulator_2]
  properties {attenuationSettings=&Fort_Weapon_Projectile_Distant_Mono_Athena_Att}
 }
 {
  id=Attenuation_5
  children [Modulator_4]
  properties {attenuationSettings=&Fort_Weapon_Projectile_Medium_Mono_Athena_Att}
 }
 {id=Mixer_2, children [Modulator_3,SoundClass_4], properties {inputVolume [1,2.5]}}
 {id=Mixer_3, children [SoundClass_3,SoundClass_5], properties {inputVolume [1,1.8]}}
 {id=Modulator_0, children [QualityLevel_2], properties {pitchMax=1.02, pitchMin=0.98, volumeMax=0.8, volumeMin=0.8}}
 {id=Modulator_2, children [QualityLevel_0], properties {volumeMax=0.25, volumeMin=0.25}}
 {id=Modulator_3, children [Mixer_3]}
 {id=Modulator_4, children [QualityLevel_1], properties {pitchMax=0.9, pitchMin=0.85, volumeMax=0.2, volumeMin=0.2}}
 {id=QualityLevel_0, children [Random_0,Random_0,Random_3,Random_3]}
 {id=QualityLevel_1, children [Random_1,Random_1,Random_4,Random_4]}
 {id=QualityLevel_2, children [Random_2,Random_2,Random_5,Random_5]}
 {id=Random_0, children [WavePlayer_18, WavePlayer_19, WavePlayer_20, WavePlayer_21], properties {weights [1,1,1,1]}}
 {
  id=Random_1
  children [WavePlayer_24, WavePlayer_25, WavePlayer_26, WavePlayer_27, WavePlayer_28, WavePlayer_29]
  properties {weights [1,1,1,1,1,1]}
 }
 {id=Random_2, children [WavePlayer_0,WavePlayer_1,WavePlayer_2,WavePlayer_3,WavePlayer_4,WavePlayer_5], properties {weights [1,1,1,1,1,1]}}
 {id=Random_3, children [WavePlayer_20, WavePlayer_21], properties {weights [1,1]}}
 {id=Random_4, children [WavePlayer_27, WavePlayer_28, WavePlayer_29], properties {weights [1,1,1]}}
 {id=Random_5, children [WavePlayer_3,WavePlayer_4,WavePlayer_5], properties {weights [1,1,1]}}
 {
  id=SoundClass_3
  children [Attenuation_3]
  properties {soundClassOverride=&Fort_WeaponRanged_Fire_3P_PSM}
 }
 {
  id=SoundClass_4
  children [Attenuation_4]
  properties {soundClassOverride=&Fort_WeaponRanged_Fire_Distant}
 }
 {
  id=SoundClass_5
  children [Attenuation_5]
  properties {soundClassOverride=&Fort_WeaponRanged_Fire_Distant}
 }
 {id=WavePlayer_0, properties {soundWaveAssetPtr=&/Ex-CaliberDMR_Fire_3P_Close_01}}
 {id=WavePlayer_1, properties {soundWaveAssetPtr=&/Ex-CaliberDMR_Fire_3P_Close_02}}
 {id=WavePlayer_18, properties {soundWaveAssetPtr=&/ExCaliberDMR_Fire_3P_Distant_01}}
 {id=WavePlayer_19, properties {soundWaveAssetPtr=&/ExCaliberDMR_Fire_3P_Distant_02}}
 {id=WavePlayer_2, properties {soundWaveAssetPtr=&/Ex-CaliberDMR_Fire_3P_Close_03}}
 {id=WavePlayer_20, properties {soundWaveAssetPtr=&/ExCaliberDMR_Fire_3P_Distant_03}}
 {id=WavePlayer_21, properties {soundWaveAssetPtr=&/ExCaliberDMR_Fire_3P_Distant_04}}
 {id=WavePlayer_24, properties {soundWaveAssetPtr=&/ExCaliberDMR_Fire_3P_Medium_01}}
 {id=WavePlayer_25, properties {soundWaveAssetPtr=&/ExCaliberDMR_Fire_3P_Medium_02}}
 {id=WavePlayer_26, properties {soundWaveAssetPtr=&/ExCaliberDMR_Fire_3P_Medium_03}}
 {id=WavePlayer_27, properties {soundWaveAssetPtr=&/ExCaliberDMR_Fire_3P_Medium_04}}
 {id=WavePlayer_28, properties {soundWaveAssetPtr=&/ExCaliberDMR_Fire_3P_Medium_05}}
 {id=WavePlayer_29, properties {soundWaveAssetPtr=&/ExCaliberDMR_Fire_3P_Medium_06}}
 {id=WavePlayer_3, properties {soundWaveAssetPtr=&/Ex-CaliberDMR_Fire_3P_Close_04}}
 {id=WavePlayer_4, properties {soundWaveAssetPtr=&/Ex-CaliberDMR_Fire_3P_Close_05}}
 {id=WavePlayer_5, properties {soundWaveAssetPtr=&/Ex-CaliberDMR_Fire_3P_Close_06}}
]
properties {
 bDebug=true
 maxDistance=27500
 soundClassObject=&Fort_WeaponRanged_Fire_3P_PSM
 volumeMultiplier=1
 concurrencyOverrides {bLimitToOwner=true, maxCount=3, resolutionRule=StopOldest}
 concurrencySet [&Weapon_Gunfire_Global_Conc, &Weapon_Gunfire_Local_Conc]
}
