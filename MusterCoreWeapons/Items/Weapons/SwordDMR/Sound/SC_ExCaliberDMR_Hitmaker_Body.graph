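# Sound cue for the ExCaliber DMR body hit marker, using an instigator 2D hit-notification sound class.
# A Mixer layers randomly chosen weapon dirt impacts, quieter flesh impacts and generic body hit-marker waves, with Modulators varying the dirt and hit-marker layers.
# A fixed quarter-second Delay at the root holds back the whole mix.
# Hit-notification concurrency is set to limit playback to two instances, stopping the oldest.
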SoundCue /MusterCoreWeapons/Items/Weapons/SwordDMR/Sound/SC_ExCaliberDMR_Hitmaker_Body
root=Delay_0
nodes [
 {id=Delay_0, children [Mixer_0], properties {delayMax=0.25, delayMin=0.25}}
 {id=Mixer_0, children [Modulator_8,Random_1,Modulator_0], properties {inputVolume [0.9,0.3,1]}}
 {id=Modulator_0, children [Random_0]}
 {id=Modulator_8, children [Random_2]}
 {id=Random_0, children [WavePlayer_0,WavePlayer_1,WavePlayer_2], properties {bShouldExcludeFromBranchCulling=true, weights [1,1,1]}}
 {id=Random_1, children [WavePlayer_4,WavePlayer_5,WavePlayer_6,WavePlayer_7], properties {weights [1,1,1,1]}}
 {
  id=Random_2
  children [WavePlayer_3, WavePlayer_8, WavePlayer_11, WavePlayer_13, WavePlayer_14, WavePlayer_15, WavePlayer_16, WavePlayer_17]
  properties {weights [1,1,1,1,1,1,1,1]}
 }
 {id=WavePlayer_0, properties {soundWaveAssetPtr=/Game/Athena/Sounds/HitNotifications/HitMarker_Body_01}}
 {id=WavePlayer_1, properties {soundWaveAssetPtr=/Game/Athena/Sounds/HitNotifications/HitMarker_Body_02}}
 {id=WavePlayer_11, properties {soundWaveAssetPtr=&/ExCaliberDMR_BulletImpact_Dirt_03}}
 {id=WavePlayer_13, properties {soundWaveAssetPtr=&/ExCaliberDMR_BulletImpact_Dirt_04}}
 {id=WavePlayer_14, properties {soundWaveAssetPtr=&/ExCaliberDMR_BulletImpact_Dirt_05}}
 {id=WavePlayer_15, properties {soundWaveAssetPtr=&/ExCaliberDMR_BulletImpact_Dirt_06}}
 {id=WavePlayer_16, properties {soundWaveAssetPtr=&/ExCaliberDMR_BulletImpact_Dirt_07}}
 {id=WavePlayer_17, properties {soundWaveAssetPtr=&/ExCaliberDMR_BulletImpact_Dirt_08}}
 {id=WavePlayer_2, properties {soundWaveAssetPtr=/Game/Athena/Sounds/HitNotifications/HitMarker_Body_03}}
 {id=WavePlayer_3, properties {soundWaveAssetPtr=&/ExCaliberDMR_BulletImpact_Dirt_01}}
 {id=WavePlayer_4, properties {soundWaveAssetPtr=/Game/Sounds/Fort_Impact_Sounds/Flesh/new_player_impact_nonspat_A_01n}}
 {id=WavePlayer_5, properties {soundWaveAssetPtr=/Game/Sounds/Fort_Impact_Sounds/Flesh/new_player_impact_nonspat_A_02n}}
 {id=WavePlayer_6, properties {soundWaveAssetPtr=/Game/Sounds/Fort_Impact_Sounds/Flesh/new_player_impact_nonspat_A_03n}}
 {id=WavePlayer_7, properties {soundWaveAssetPtr=/Game/Sounds/Fort_Impact_Sounds/Flesh/new_player_impact_nonspat_A_04n}}
 {id=WavePlayer_8, properties {soundWaveAssetPtr=&/ExCaliberDMR_BulletImpact_Dirt_02}}
]
properties {
 maxDistance=2097152
 soundClassObject=&Fort_2DHitNotifications_Instigate_PSM
 volumeMultiplier=0.757
 concurrencyOverrides {maxCount=2, resolutionRule=StopOldest}
 concurrencySet [&HitNotifications_TwoMaximum_StopOldest_VolumeScale]
}
