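# Sound cue for the ExCaliber DMR's flesh impact, mixing a bullet-impact layer with a louder random sweetener layer.
# Each layer switches by quality level: the first two use full variation sets, the last two use smaller ones.
# The full impact layer blends random dirt impacts with quieter shared player flesh hits; the reduced one plays only pitch- and volume-modulated dirt.
# Overridden concurrency allows one instance per owner, preventing new ones.
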SoundCue /MusterCoreWeapons/Items/Weapons/SwordDMR/Sound/SC_ExCaliberDMR_Impact_Flesh
root=Mixer_1
nodes [
 {id=Mixer_0, children [Modulator_8,Random_1], properties {inputVolume [1,0.5]}}
 {id=Mixer_1, children [QualityLevel_1, QualityLevel_2], properties {inputVolume [1,1.5]}}
 {id=Modulator_1, children [Random_3], properties {pitchMax=1.25, pitchMin=1, volumeMax=1.1, volumeMin=0.8}}
 {id=Modulator_8, children [Random_2]}
 {id=QualityLevel_1, children [Mixer_0,Mixer_0,Modulator_1,Modulator_1]}
 {id=QualityLevel_2, children [Random_0,Random_0,Random_4,Random_4]}
 {id=Random_0, children [WavePlayer_0, WavePlayer_1, WavePlayer_8, WavePlayer_9, WavePlayer_10], properties {weights [1,1,1,1,1]}}
 {id=Random_1, children [WavePlayer_4,WavePlayer_5,WavePlayer_6,WavePlayer_7], properties {weights [1,1,1,1]}}
 {
  id=Random_2
  children [WavePlayer_2, WavePlayer_3, WavePlayer_11, WavePlayer_13, WavePlayer_14, WavePlayer_15, WavePlayer_16, WavePlayer_17]
  properties {weights [1,1,1,1,1,1,1,1]}
 }
 {id=Random_3, children [WavePlayer_18, WavePlayer_19, WavePlayer_20], properties {weights [1,1,1]}}
 {id=Random_4, children [WavePlayer_12, WavePlayer_21], properties {weights [1,1]}}
 {id=WavePlayer_0, properties {soundWaveAssetPtr=&/ExCaliberDMR_Impact_Sweetener_01}}
 {id=WavePlayer_1, properties {soundWaveAssetPtr=&/ExCaliberDMR_Impact_Sweetener_02}}
 {id=WavePlayer_10, properties {soundWaveAssetPtr=&/ExCaliberDMR_Impact_Sweetener_05}}
 {id=WavePlayer_11, properties {soundWaveAssetPtr=&/ExCaliberDMR_BulletImpact_Dirt_03}}
 {id=WavePlayer_12, properties {soundWaveAssetPtr=&ExCaliberDMR_Impact_Sweetener_02}}
 {id=WavePlayer_13, properties {soundWaveAssetPtr=&/ExCaliberDMR_BulletImpact_Dirt_04}}
 {id=WavePlayer_14, properties {soundWaveAssetPtr=&/ExCaliberDMR_BulletImpact_Dirt_05}}
 {id=WavePlayer_15, properties {soundWaveAssetPtr=&/ExCaliberDMR_BulletImpact_Dirt_06}}
 {id=WavePlayer_16, properties {soundWaveAssetPtr=&/ExCaliberDMR_BulletImpact_Dirt_07}}
 {id=WavePlayer_17, properties {soundWaveAssetPtr=&/ExCaliberDMR_BulletImpact_Dirt_08}}
 {id=WavePlayer_18, properties {soundWaveAssetPtr=&ExCaliberDMR_BulletImpact_Dirt_06}}
 {id=WavePlayer_19, properties {soundWaveAssetPtr=&ExCaliberDMR_BulletImpact_Dirt_07}}
 {id=WavePlayer_2, properties {soundWaveAssetPtr=&/ExCaliberDMR_BulletImpact_Dirt_01}}
 {id=WavePlayer_20, properties {soundWaveAssetPtr=&ExCaliberDMR_BulletImpact_Dirt_08}}
 {id=WavePlayer_21, properties {soundWaveAssetPtr=&ExCaliberDMR_Impact_Sweetener_05}}
 {id=WavePlayer_3, properties {soundWaveAssetPtr=&/ExCaliberDMR_BulletImpact_Dirt_02}}
 {id=WavePlayer_4, properties {soundWaveAssetPtr=/Game/Sounds/Fort_Impact_Sounds/Flesh/new_player_impact_nonspat_A_01n}}
 {id=WavePlayer_5, properties {soundWaveAssetPtr=/Game/Sounds/Fort_Impact_Sounds/Flesh/new_player_impact_nonspat_A_02n}}
 {id=WavePlayer_6, properties {soundWaveAssetPtr=/Game/Sounds/Fort_Impact_Sounds/Flesh/new_player_impact_nonspat_A_03n}}
 {id=WavePlayer_7, properties {soundWaveAssetPtr=/Game/Sounds/Fort_Impact_Sounds/Flesh/new_player_impact_nonspat_A_04n}}
 {id=WavePlayer_8, properties {soundWaveAssetPtr=&/ExCaliberDMR_Impact_Sweetener_03}}
 {id=WavePlayer_9, properties {soundWaveAssetPtr=&/ExCaliberDMR_Impact_Sweetener_04}}
]
properties {
 attenuationSettings=/Game/Sounds/Attenuation/Fort_Weapon_Pickaxe_Player_Impact_Att
 bOverrideConcurrency=true
 maxDistance=3512
 soundClassObject=&Fort_Impacts
 volumeMultiplier=1.875
 attenuationOverrides {
  bAttenuate=false
  bAttenuateWithLPF=true
  distanceAlgorithm=Logarithmic
  falloffDistance=800
  lPFRadiusMax=800
  lPFRadiusMin=400
  attenuationShapeExtents {x=300}
 }
 concurrencyOverrides {bLimitToOwner=true, maxCount=1, resolutionRule=PreventNew}
 concurrencySet [/Game/Sounds/Concurrency_Rules/OneMaximum_Owner_StopOldest]
}
