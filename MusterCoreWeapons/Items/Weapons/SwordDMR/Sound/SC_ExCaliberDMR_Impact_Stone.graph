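# Sound cue for ExCaliber DMR bullet impacts on stone, built from wood impact waves plus sweetener layers.
# After a short random delay, a mixer layers a pitch- and volume-modulated random impact with a random sweetener.
# Quality-level switches on both layers give the first two levels full variation pools and the last two smaller subsets.
# It uses bullet-impact attenuation, the impacts sound class and rifle impact concurrency.
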SoundCue /MusterCoreWeapons/Items/Weapons/SwordDMR/Sound/SC_ExCaliberDMR_Impact_Stone
root=Delay_0
nodes [
 {id=Delay_0, children [Mixer_0], properties {delayMax=0.13, delayMin=0.07}}
 {id=Mixer_0, children [QualityLevel_1, QualityLevel_2], properties {inputVolume [1,1]}}
 {id=Modulator_1, children [Random_3], properties {pitchMax=1.25, pitchMin=1, volumeMax=1.1, volumeMin=0.8}}
 {id=Modulator_2, children [Random_2], properties {pitchMax=1.25, pitchMin=1, volumeMax=1.1, volumeMin=0.8}}
 {id=QualityLevel_1, children [Modulator_2,Modulator_2,Modulator_1,Modulator_1]}
 {id=QualityLevel_2, children [Random_0,Random_0,Random_1,Random_1]}
 {id=Random_0, children [WavePlayer_0,WavePlayer_1,WavePlayer_2,WavePlayer_3,WavePlayer_4], properties {weights [1,1,1,1,1]}}
 {id=Random_1, children [WavePlayer_5,WavePlayer_6], properties {weights [1,1]}}
 {
  id=Random_2
  children [WavePlayer_22, WavePlayer_23, WavePlayer_24, WavePlayer_25, WavePlayer_26, WavePlayer_27, WavePlayer_28, WavePlayer_29]
  properties {weights [1,1,1,1,1,1,1,1]}
 }
 {id=Random_3, children [WavePlayer_30, WavePlayer_31, WavePlayer_32], properties {weights [1,1,1]}}
 {id=WavePlayer_0, properties {soundWaveAssetPtr=&/ExCaliberDMR_Impact_Sweetener_01}}
 {id=WavePlayer_1, properties {soundWaveAssetPtr=&/ExCaliberDMR_Impact_Sweetener_02}}
 {id=WavePlayer_2, properties {soundWaveAssetPtr=&/ExCaliberDMR_Impact_Sweetener_03}}
 {id=WavePlayer_22, properties {soundWaveAssetPtr=&/ExCaliberDMR_BulletImpact_Wood_01}}
 {id=WavePlayer_23, properties {soundWaveAssetPtr=&/ExCaliberDMR_BulletImpact_Wood_02}}
 {id=WavePlayer_24, properties {soundWaveAssetPtr=&/ExCaliberDMR_BulletImpact_Wood_03}}
 {id=WavePlayer_25, properties {soundWaveAssetPtr=&/ExCaliberDMR_BulletImpact_Wood_04}}
 {id=WavePlayer_26, properties {soundWaveAssetPtr=&/ExCaliberDMR_BulletImpact_Wood_05}}
 {id=WavePlayer_27, properties {soundWaveAssetPtr=&/ExCaliberDMR_BulletImpact_Wood_06}}
 {id=WavePlayer_28, properties {soundWaveAssetPtr=&/ExCaliberDMR_BulletImpact_Wood_07}}
 {id=WavePlayer_29, properties {soundWaveAssetPtr=&/ExCaliberDMR_BulletImpact_Wood_08}}
 {id=WavePlayer_3, properties {soundWaveAssetPtr=&/ExCaliberDMR_Impact_Sweetener_04}}
 {id=WavePlayer_30, properties {soundWaveAssetPtr=&ExCaliberDMR_BulletImpact_Wood_06}}
 {id=WavePlayer_31, properties {soundWaveAssetPtr=&ExCaliberDMR_BulletImpact_Wood_07}}
 {id=WavePlayer_32, properties {soundWaveAssetPtr=&ExCaliberDMR_BulletImpact_Wood_08}}
 {id=WavePlayer_4, properties {soundWaveAssetPtr=&/ExCaliberDMR_Impact_Sweetener_05}}
 {id=WavePlayer_5, properties {soundWaveAssetPtr=&ExCaliberDMR_Impact_Sweetener_02}}
 {id=WavePlayer_6, properties {soundWaveAssetPtr=&ExCaliberDMR_Impact_Sweetener_05}}
]
properties {
 attenuationSettings=/Game/Sounds/Attenuation/Fort_Bullet_Impact_Att
 maxDistance=5300
 soundClassObject=&Fort_Impacts
 volumeMultiplier=2.5
 concurrencyOverrides {maxCount=4, resolutionRule=PreventNew}
 concurrencySet [&RifleBulletImpact_Conc]
}
