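# Sound cue for the Ex-Caliber DMR's no-ammo sound.
# A single wave player feeds a Modulator that pins pitch at unity, while cue-level multipliers slightly lower pitch and reduce volume.
# It uses the third-person ranged weapon fire sound class, carries a falloff distance setting, and overrides concurrency to allow two instances, stopping the oldest.
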SoundCue /MusterCoreWeapons/Items/Weapons/SwordDMR/Sound/SC_ExCaliberDMR_NoAmmo
root=Modulator_1
nodes [
 {id=Modulator_1, children [WavePlayer_0], properties {pitchMax=1, pitchMin=1}}
 {id=WavePlayer_0, properties {soundWaveAssetPtr=&/Ex-CaliberDMR_NoAmmo}}
]
properties {
 bOverrideConcurrency=true
 maxDistance=2097152
 pitchMultiplier=0.95
 soundClassObject=&Fort_WeaponRanged_Fire_3P_PSM
 volumeMultiplier=0.65
 attenuationOverrides {falloffDistance=1600}
 concurrencyOverrides {maxCount=2, resolutionRule=StopOldest}
}
