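# Sound cue for the Ex-Caliber DMR's chamber-opening reload foley.
# It mixes a random open-chamber variant, a quiet reload-start wave, and a random shell drop delayed by one second.
# A quality-level switch gives the first two levels six shell-drop variants and the last two only three.
# The mix routes through the weapon foley sound class and slight volume modulation, with overridden attenuation and debug left on.
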SoundCue /MusterCoreWeapons/Items/Weapons/SwordDMR/Sound/SC_ExCaliberDMR_OpenChamber
root=Modulator_0
nodes [
 {id=Delay_0, children [QualityLevel_0], properties {delayMax=1, delayMin=1}}
 {id=Mixer_0, children [Random_1,Delay_0,WavePlayer_0], properties {inputVolume [1,1.5,0.15]}}
 {id=Modulator_0, children [SoundClass_0], properties {pitchMax=1, pitchMin=1, volumeMax=0.6, volumeMin=0.5}}
 {id=QualityLevel_0, children [Random_0,Random_0,Random_2,Random_2]}
 {id=Random_0, children [WavePlayer_1,WavePlayer_2,WavePlayer_3,WavePlayer_4,WavePlayer_5,WavePlayer_6], properties {weights [1,1,1,1,1,1]}}
 {id=Random_1, children [WavePlayer_7, WavePlayer_8, WavePlayer_9, WavePlayer_10], properties {weights [1,1,1,1]}}
 {id=Random_2, children [WavePlayer_4,WavePlayer_5,WavePlayer_6], properties {weights [1,1,1]}}
 {id=SoundClass_0, children [Mixer_0], properties {soundClassOverride=&Fort_WeaponFoley}}
 {id=WavePlayer_0, properties {soundWaveAssetPtr=&/Ex-CaliberDMR_Reload_Start}}
 {id=WavePlayer_1, properties {soundWaveAssetPtr=&/Ex-CaliberDMR_ReloadShellDrop_01}}
 {id=WavePlayer_10, properties {soundWaveAssetPtr=&/Ex-CaliberDMR_OpenChamber_03}}
 {id=WavePlayer_2, properties {soundWaveAssetPtr=&/Ex-CaliberDMR_ReloadShellDrop_02}}
 {id=WavePlayer_3, properties {soundWaveAssetPtr=&/Ex-CaliberDMR_ReloadShellDrop_03}}
 {id=WavePlayer_4, properties {soundWaveAssetPtr=&/Ex-CaliberDMR_ReloadShellDrop_04}}
 {id=WavePlayer_5, properties {soundWaveAssetPtr=&/Ex-CaliberDMR_ReloadShellDrop_05}}
 {id=WavePlayer_6, properties {soundWaveAssetPtr=&/Ex-CaliberDMR_ReloadShellDrop_06}}
 {id=WavePlayer_7, properties {soundWaveAssetPtr=&/Ex-CaliberDMR_OpenChamber_04}}
 {id=WavePlayer_8, properties {soundWaveAssetPtr=&/Ex-CaliberDMR_OpenChamber_01}}
 {id=WavePlayer_9, properties {soundWaveAssetPtr=&/Ex-CaliberDMR_OpenChamber_02}}
]
properties {
 attenuationSettings=&Fort_Weapon_Reload_Att
 bDebug=true
 bOverrideAttenuation=true
 bOverrideConcurrency=true
 maxDistance=2000
 soundClassObject=&Fort_WeaponFoley
 volumeMultiplier=1.578
 attenuationOverrides {falloffDistance=1500, stereoSpread=300, attenuationShapeExtents {x=500}}
}
