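# Sound cue for picking up the ExCaliber DMR sword weapon, played through loot-collect class and weapon-pickup attenuation.
# A mixer layers the main pickup wave with a louder energy wave that is slightly delayed and randomly pitched up.
# All quality levels share that same mix, and the whole cue plays quieter and lower in pitch.
# Concurrency uses a pickup-maximum set, alongside override values allowing one instance and preventing new ones.
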SoundCue /MusterCoreWeapons/Items/Weapons/SwordDMR/Sound/SC_ExCaliberDMR_Pickup
root=QualityLevel_0
nodes [
 {id=Delay_0, children [Modulator_0], properties {delayMax=0.1, delayMin=0.1}}
 {id=Mixer_2, children [WavePlayer_6,Delay_0], properties {inputVolume [1,1.5]}}
 {id=Modulator_0, children [WavePlayer_4], properties {pitchMax=1.2, pitchMin=1}}
 {id=QualityLevel_0, children [Mixer_2,Mixer_2,Mixer_2,Mixer_2]}
 {id=WavePlayer_4, properties {soundWaveAssetPtr=&/ExCaliberDMR_PickupEnergy_03}}
 {id=WavePlayer_6, properties {soundWaveAssetPtr=&/ExCaliberDMR_PickupNew}}
]
properties {
 attenuationSettings=&Fort_Weapon_Pickup_Att
 maxDistance=3500
 pitchMultiplier=0.9
 soundClassObject=&Fort_Loot_Collect
 volumeMultiplier=0.2
 concurrencyOverrides {maxCount=1, resolutionRule=PreventNew}
 concurrencySet [&PickupMaximum_Oldest]
}
