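# Sound cue for the countdown sound of the Ex-Caliber DMR (SwordDMR) weapon's projectile.
# A Random node picks among seven equally weighted countdown wave variations, then a Modulator at its default settings feeds the output.
# It plays on the weapons sound class with a boosted volume and the explosive-countdown attenuation asset for distance falloff.
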SoundCue /MusterCoreWeapons/Items/Weapons/SwordDMR/Sound/SC_ExCaliberDMR_Proj_Countdown
root=Modulator_0
nodes [
 {id=Modulator_0, children [Random_0]}
 {
  id=Random_0
  children [WavePlayer_0,WavePlayer_1,WavePlayer_2,WavePlayer_3,WavePlayer_4,WavePlayer_5,WavePlayer_6]
  properties {weights [1,1,1,1,1,1,1]}
 }
 {id=WavePlayer_0, properties {soundWaveAssetPtr=&/Ex-CaliberDMR_ProjCountdown_01}}
 {id=WavePlayer_1, properties {soundWaveAssetPtr=&/Ex-CaliberDMR_ProjCountdown_02}}
 {id=WavePlayer_2, properties {soundWaveAssetPtr=&/Ex-CaliberDMR_ProjCountdown_03}}
 {id=WavePlayer_3, properties {soundWaveAssetPtr=&/Ex-CaliberDMR_ProjCountdown_04}}
 {id=WavePlayer_4, properties {soundWaveAssetPtr=&/Ex-CaliberDMR_ProjCountdown_05}}
 {id=WavePlayer_5, properties {soundWaveAssetPtr=&/Ex-CaliberDMR_ProjCountdown_06}}
 {id=WavePlayer_6, properties {soundWaveAssetPtr=&/Ex-CaliberDMR_ProjCountdown_07}}
]
properties {
 attenuationSettings=/Game/Sounds/Attenuation/Fort_ExCaliberExplosiveCountdown_Att
 maxDistance=10200
 soundClassObject=/Game/Sounds/FortSoundClassesAndModes/Fort_Weapons
 volumeMultiplier=3
}
