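# Sound cue for the Ex-Caliber DMR projectile explosion, mixed from close, medium and distant layers with their own attenuation.
# Each layer blends quieter generic Clinger or grenade explosions with random weapon-specific samples; medium reuses the close samples.
# A briefly delayed close-attenuated buzz joins the close and medium mixes.
# Sound class overrides separate near from distant layers, quality nodes switch between full and reduced variation pools, and concurrency stops the oldest instance.
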SoundCue /MusterCoreWeapons/Items/Weapons/SwordDMR/Sound/SC_ExCaliberDMR_Proj_Explode
root=Mixer_4
nodes [
 {
  id=Attenuation_0
  children [Modulator_1]
  properties {
   attenuationSettings=/Game/Sounds/Attenuation/Fort_Explosion_Distant_Athena_Att
   attenuationOverrides {falloffDistance=3000, attenuationShapeExtents {x=1000}}
  }
 }
 {
  id=Attenuation_1
  children [QualityLevel_0]
  properties {attenuationSettings=/Game/Sounds/Attenuation/Fort_Explosion_Close_Athena_Att, attenuationOverrides {falloffDistance=1600}}
 }
 {
  id=Attenuation_2
  children [Modulator_3]
  properties {
   attenuationSettings=/Game/Sounds/Attenuation/Fort_Explosion_Medium_Athena_Att
   attenuationOverrides {falloffDistance=3000, attenuationShapeExtents {x=1000}}
  }
 }
 {
  id=Attenuation_3
  children [Modulator_4]
  properties {
   attenuationSettings=&Fort_Explosion_Distant_Athena_Att
   attenuationOverrides {falloffDistance=3000, attenuationShapeExtents {x=1000}}
  }
 }
 {
  id=Attenuation_4
  children [QualityLevel_3]
  properties {attenuationSettings=&Fort_Explosion_Close_Athena_Att, attenuationOverrides {falloffDistance=1600}}
 }
 {
  id=Attenuation_5
  children [Modulator_5]
  properties {
   attenuationSettings=&Fort_Explosion_Medium_Athena_Att
   attenuationOverrides {falloffDistance=3000, attenuationShapeExtents {x=1000}}
  }
 }
 {
  id=Attenuation_6
  children [Delay_0]
  properties {attenuationSettings=&Fort_Explosion_Close_Athena_Att, attenuationOverrides {falloffDistance=1600}}
 }
 {
  id=Attenuation_7
  children [Delay_1]
  properties {attenuationSettings=&Fort_Explosion_Close_Athena_Att, attenuationOverrides {falloffDistance=1600}}
 }
 {id=Delay_0, children [WavePlayer_33], properties {delayMax=0.1, delayMin=0.1}}
 {id=Delay_1, children [WavePlayer_36], properties {delayMax=0.1, delayMin=0.1}}
 {id=Mixer_0, children [SoundClass_3,SoundClass_1,SoundClass_6], properties {inputVolume [0.2,1,0.7]}}
 {id=Mixer_1, children [SoundClass_2,SoundClass_4,SoundClass_7], properties {inputVolume [0.2,1,0.8]}}
 {id=Mixer_2, children [SoundClass_0,SoundClass_5,null], properties {inputVolume [0.2,1,1]}}
 {id=Mixer_4, children [Mixer_0,Mixer_1,Mixer_2], properties {inputVolume [2.5,0.8,0.5]}}
 {id=Modulator_0, children [Attenuation_4]}
 {id=Modulator_1, children [QualityLevel_5], properties {pitchMax=1.1, pitchMin=0.9}}
 {id=Modulator_2, children [Attenuation_1]}
 {id=Modulator_3, children [QualityLevel_1], properties {pitchMax=1.1, pitchMin=0.9}}
 {id=Modulator_4, children [QualityLevel_2], properties {pitchMax=1.1, pitchMin=0.9}}
 {id=Modulator_5, children [QualityLevel_4], properties {pitchMax=1.1, pitchMin=0.9}}
 {id=Modulator_6, children [Attenuation_6], properties {pitchMax=1.2, pitchMin=1, volumeMax=1, volumeMin=1}}
 {id=Modulator_7, children [Attenuation_7], properties {pitchMax=1, pitchMin=1}}
 {id=QualityLevel_0, children [Random_3,Random_3,Random_8,Random_8]}
 {id=QualityLevel_1, children [Random_4,Random_4,Random_9,Random_9]}
 {id=QualityLevel_2, children [Random_5,Random_5,Random_10,Random_10]}
 {id=QualityLevel_3, children [Random_0,Random_0,Random_11,Random_11]}
 {id=QualityLevel_4, children [Random_1,Random_1,Random_12,Random_12]}
 {id=QualityLevel_5, children [Random_2,Random_2,Random_13,Random_13]}
 {id=Random_0, children [WavePlayer_0,WavePlayer_1,WavePlayer_4], properties {weights [1,1,1]}}
 {id=Random_1, children [WavePlayer_12, WavePlayer_13, WavePlayer_14], properties {weights [1,1,1]}}
 {id=Random_10, children [WavePlayer_27, WavePlayer_28, WavePlayer_29], properties {weights [1,1,1]}}
 {id=Random_11, children [WavePlayer_1,WavePlayer_4], properties {weights [1,1]}}
 {id=Random_12, children [WavePlayer_13, WavePlayer_14], properties {weights [1,1]}}
 {id=Random_13, children [WavePlayer_16, WavePlayer_17], properties {weights [1,1]}}
 {id=Random_2, children [WavePlayer_15, WavePlayer_16, WavePlayer_17], properties {weights [1,1,1]}}
 {id=Random_3, children [WavePlayer_2,WavePlayer_3,WavePlayer_5,WavePlayer_6,WavePlayer_7,WavePlayer_8], properties {weights [1,1,1,1,1,1]}}
 {
  id=Random_4
  children [WavePlayer_11, WavePlayer_18, WavePlayer_19, WavePlayer_20, WavePlayer_21, WavePlayer_22]
  properties {weights [1,1,1,1,1,1]}
 }
 {id=Random_5, children [WavePlayer_25, WavePlayer_26, WavePlayer_27, WavePlayer_28, WavePlayer_29], properties {weights [1,1,1,1,1]}}
 {id=Random_8, children [WavePlayer_6,WavePlayer_7,WavePlayer_8], properties {weights [1,1,1]}}
 {id=Random_9, children [WavePlayer_20, WavePlayer_21, WavePlayer_22], properties {weights [1,1,1]}}
 {
  id=SoundClass_0
  children [Attenuation_0]
  properties {soundClassOverride=&Fort_WeaponRanged_Fire_Distant}
 }
 {id=SoundClass_1, children [Modulator_2], properties {soundClassOverride=&Fort_WeaponRanged_Fire_3P_PSM}}
 {
  id=SoundClass_2
  children [Attenuation_5]
  properties {soundClassOverride=&Fort_WeaponRanged_Fire_Distant}
 }
 {id=SoundClass_3, children [Modulator_0], properties {soundClassOverride=&Fort_WeaponRanged_Fire_3P_PSM}}
 {
  id=SoundClass_4
  children [Attenuation_2]
  properties {soundClassOverride=&Fort_WeaponRanged_Fire_Distant}
 }
 {
  id=SoundClass_5
  children [Attenuation_3]
  properties {soundClassOverride=&Fort_WeaponRanged_Fire_Distant}
 }
 {id=SoundClass_6, children [Modulator_6], properties {soundClassOverride=&Fort_WeaponRanged_Fire_3P_PSM}}
 {id=SoundClass_7, children [Modulator_7], properties {soundClassOverride=&Fort_WeaponRanged_Fire_3P_PSM}}
 {id=WavePlayer_0, properties {soundWaveAssetPtr=/Game/Athena/Sounds/Weapons/Explosions/Clinger_Expl_Close_02}}
 {id=WavePlayer_1, properties {soundWaveAssetPtr=/Game/Athena/Sounds/Weapons/Explosions/Clinger_Expl_Close_01}}
 {id=WavePlayer_11, properties {soundWaveAssetPtr=&/Ex-CaliberDMR_ProjectileExplode_Close_01}}
 {id=WavePlayer_12, properties {soundWaveAssetPtr=/Game/Athena/Sounds/Weapons/Explosions/Clinger_Expl_Med_02}}
 {id=WavePlayer_13, properties {soundWaveAssetPtr=/Game/Athena/Sounds/Weapons/Explosions/Clinger_Expl_Med_03}}
 {id=WavePlayer_14, properties {soundWaveAssetPtr=/Game/Athena/Sounds/Weapons/Explosions/Clinger_Expl_Med_01}}
 {id=WavePlayer_15, properties {soundWaveAssetPtr=/Game/Athena/Sounds/Weapons/Explosions/Grenade_Expl_Dist_01}}
 {id=WavePlayer_16, properties {soundWaveAssetPtr=/Game/Athena/Sounds/Weapons/Explosions/Grenade_Expl_Dist_02}}
 {id=WavePlayer_17, properties {soundWaveAssetPtr=/Game/Athena/Sounds/Weapons/Explosions/Grenade_Expl_Dist_03}}
 {id=WavePlayer_18, properties {soundWaveAssetPtr=&/Ex-CaliberDMR_ProjectileExplode_Close_02}}
 {id=WavePlayer_19, properties {soundWaveAssetPtr=&/Ex-CaliberDMR_ProjectileExplode_Close_03}}
 {id=WavePlayer_2, properties {soundWaveAssetPtr=&Ex-CaliberDMR_ProjectileExplode_Close_01}}
 {id=WavePlayer_20, properties {soundWaveAssetPtr=&/Ex-CaliberDMR_ProjectileExplode_Close_04}}
 {id=WavePlayer_21, properties {soundWaveAssetPtr=&/Ex-CaliberDMR_ProjectileExplode_Close_05}}
 {id=WavePlayer_22, properties {soundWaveAssetPtr=&/Ex-CaliberDMR_ProjectileExplode_Close_06}}
 {id=WavePlayer_25, properties {soundWaveAssetPtr=&/Ex-CaliberDMR_ProjectileExplode_Distant_01}}
 {id=WavePlayer_26, properties {soundWaveAssetPtr=&/Ex-CaliberDMR_ProjectileExplode_Distant_02}}
 {id=WavePlayer_27, properties {soundWaveAssetPtr=&/Ex-CaliberDMR_ProjectileExplode_Distant_03}}
 {id=WavePlayer_28, properties {soundWaveAssetPtr=&/Ex-CaliberDMR_ProjectileExplode_Distant_04}}
 {id=WavePlayer_29, properties {soundWaveAssetPtr=&/Ex-CaliberDMR_ProjectileExplode_Distant_05}}
 {id=WavePlayer_3, properties {soundWaveAssetPtr=&Ex-CaliberDMR_ProjectileExplode_Close_02}}
 {id=WavePlayer_33, properties {soundWaveAssetPtr=&/Ex-CaliberDMR_ProjectileExplode_Buzz_01}}
 {id=WavePlayer_36, properties {soundWaveAssetPtr=&Ex-CaliberDMR_ProjectileExplode_Buzz_01}}
 {id=WavePlayer_4, properties {soundWaveAssetPtr=/Game/Athena/Sounds/Weapons/Explosions/Clinger_Expl_Close_03}}
 {id=WavePlayer_5, properties {soundWaveAssetPtr=&Ex-CaliberDMR_ProjectileExplode_Close_03}}
 {id=WavePlayer_6, properties {soundWaveAssetPtr=&Ex-CaliberDMR_ProjectileExplode_Close_04}}
 {id=WavePlayer_7, properties {soundWaveAssetPtr=&Ex-CaliberDMR_ProjectileExplode_Close_05}}
 {id=WavePlayer_8, properties {soundWaveAssetPtr=&Ex-CaliberDMR_ProjectileExplode_Close_06}}
]
properties {
 maxDistance=27500
 volumeMultiplier=0.8
 concurrencyOverrides {maxCount=4, resolutionRule=StopOldest, volumeScale=0.5}
 concurrencySet [/Game/Sounds/Concurrency_Rules/Weapon_Explosion_Conc]
}
