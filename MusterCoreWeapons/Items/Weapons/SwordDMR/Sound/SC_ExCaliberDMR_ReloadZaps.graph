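# Sound cue for the electric zap sounds in the ExCaliber DMR's reload.
# A Random node picks one of four equally weighted zap waves and feeds a SoundClass node that assigns the weapon foley class.
# A Modulator slightly varies volume with pitch fixed, then a root Attenuation node applies the shared weapon reload falloff.
# The cue overrides concurrency and has its debug flag switched on.
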SoundCue /MusterCoreWeapons/Items/Weapons/SwordDMR/Sound/SC_ExCaliberDMR_ReloadZaps
root=Attenuation_0
nodes [
 {id=Attenuation_0, children [Modulator_0], properties {attenuationSettings=&Fort_Weapon_Reload_Att}}
 {id=Modulator_0, children [SoundClass_0], properties {pitchMax=1, pitchMin=1, volumeMax=0.6, volumeMin=0.5}}
 {id=Random_0, children [WavePlayer_10, WavePlayer_11, WavePlayer_12, WavePlayer_13], properties {weights [1,1,1,1]}}
 {id=SoundClass_0, children [Random_0], properties {soundClassOverride=&Fort_WeaponFoley}}
 {id=WavePlayer_10, properties {soundWaveAssetPtr=&/ExCaliberDMR_ReloadZaps_01}}
 {id=WavePlayer_11, properties {soundWaveAssetPtr=&/ExCaliberDMR_ReloadZaps_02}}
 {id=WavePlayer_12, properties {soundWaveAssetPtr=&/ExCaliberDMR_ReloadZaps_03}}
 {id=WavePlayer_13, properties {soundWaveAssetPtr=&/ExCaliberDMR_ReloadZaps_04}}
]
properties {
 attenuationSettings=&Fort_Weapon_Reload_Att
 bDebug=true
 bOverrideConcurrency=true
 maxDistance=2250
 soundClassObject=&Fort_WeaponFoley
 volumeMultiplier=1.253
}
