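# Sound cue for the cartridge step of the Ex-Caliber DMR reload, classed as weapon foley.
# An equally weighted Random node picks one of three wave variations, routed through a SoundClass node and a Modulator that slightly varies and lowers volume.
# The cue slightly lowers pitch, spatializes through the shared weapon reload attenuation asset, overrides concurrency, and has its debug flag left on.
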SoundCue /MusterCoreWeapons/Items/Weapons/SwordDMR/Sound/SC_ExCaliberDMR_Reload_Cartridge
root=Modulator_0
nodes [
 {id=Modulator_0, children [SoundClass_0], properties {volumeMax=0.7, volumeMin=0.6}}
 {id=Random_0, children [WavePlayer_0,WavePlayer_1,WavePlayer_2], properties {weights [1,1,1]}}
 {id=SoundClass_0, children [Random_0], properties {soundClassOverride=&Fort_WeaponFoley}}
 {id=WavePlayer_0, properties {soundWaveAssetPtr=&/Ex-CaliberDMR_Reload_Cartridge_01}}
 {id=WavePlayer_1, properties {soundWaveAssetPtr=&/Ex-CaliberDMR_Reload_Cartridge_02}}
 {id=WavePlayer_2, properties {soundWaveAssetPtr=&/Ex-CaliberDMR_Reload_Cartridge_03}}
]
properties {
 attenuationSettings=&Fort_Weapon_Reload_Att
 bDebug=true
 bOverrideConcurrency=true
 maxDistance=2250
 pitchMultiplier=0.9
 soundClassObject=&Fort_WeaponFoley
 volumeMultiplier=0.802
 attenuationOverrides {falloffDistance=1500, stereoSpread=300, attenuationShapeExtents {x=500}}
}
