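# Sound cue for the end of the Ex-Caliber DMR reload, built from the weapon's Equip wave recordings.
# An equal-weight random node picks one of six variants, which pass through a weapon foley sound class node.
# A modulator then varies volume slightly at fixed pitch, and an attenuation node applies the shared weapon reload falloff.
# The cue overrides concurrency and has its debug flag enabled.
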SoundCue /MusterCoreWeapons/Items/Weapons/SwordDMR/Sound/SC_ExCaliberDMR_Reload_End
root=Attenuation_0
nodes [
 {id=Attenuation_0, children [Modulator_0], properties {attenuationSettings=&Fort_Weapon_Reload_Att}}
 {id=Modulator_0, children [SoundClass_0], properties {pitchMax=1, pitchMin=1, volumeMax=0.6, volumeMin=0.5}}
 {id=Random_0, children [WavePlayer_1,WavePlayer_2,WavePlayer_3,WavePlayer_4,WavePlayer_5,WavePlayer_6], properties {weights [1,1,1,1,1,1]}}
 {id=SoundClass_0, children [Random_0], properties {soundClassOverride=&Fort_WeaponFoley}}
 {id=WavePlayer_1, properties {soundWaveAssetPtr=&/Ex-CaliberDMR_Equip_01}}
 {id=WavePlayer_2, properties {soundWaveAssetPtr=&/Ex-CaliberDMR_Equip_02}}
 {id=WavePlayer_3, properties {soundWaveAssetPtr=&/Ex-CaliberDMR_Equip_03}}
 {id=WavePlayer_4, properties {soundWaveAssetPtr=&/Ex-CaliberDMR_Equip_04}}
 {id=WavePlayer_5, properties {soundWaveAssetPtr=&/Ex-CaliberDMR_Equip_05}}
 {id=WavePlayer_6, properties {soundWaveAssetPtr=&/Ex-CaliberDMR_Equip_06}}
]
properties {
 attenuationSettings=&Fort_Weapon_Reload_Att
 bDebug=true
 bOverrideConcurrency=true
 maxDistance=2250
 soundClassObject=&Fort_WeaponFoley
 volumeMultiplier=1.253
}
