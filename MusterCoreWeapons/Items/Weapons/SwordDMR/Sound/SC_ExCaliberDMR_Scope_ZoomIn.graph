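# Sound cue for the scope zoom-in of the ExCaliber DMR, played as quiet weapon foley.
# A mixer layers one of four equally weighted, randomly chosen player movement foley waves with a dedicated scope-in wave, each through its own volume modulator.
# The cue uses the weapon foley sound class with overridden attenuation.
# A concurrency override allows only one instance, stopping the oldest on retrigger.
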SoundCue /MusterCoreWeapons/Items/Weapons/SwordDMR/Sound/SC_ExCaliberDMR_Scope_ZoomIn
root=Mixer_0
nodes [
 {id=Mixer_0, children [Modulator_0,Modulator_1], properties {inputVolume [0.4,0.5]}}
 {id=Modulator_0, children [Random_4], properties {volumeMax=0.8, volumeMin=0.8}}
 {id=Modulator_1, children [WavePlayer_4], properties {pitchMax=1, volumeMax=2.5, volumeMin=2.5}}
 {id=Random_4, children [WavePlayer_0,WavePlayer_1,WavePlayer_2,WavePlayer_3], properties {weights [1,1,1,1]}}
 {id=WavePlayer_0, properties {soundWaveAssetPtr=/Game/Sounds/Fort_Foley_Sounds/Moves/player_run_foley_15}}
 {id=WavePlayer_1, properties {soundWaveAssetPtr=/Game/Sounds/Fort_Foley_Sounds/Moves/player_run_foley_14}}
 {id=WavePlayer_2, properties {soundWaveAssetPtr=/Game/Sounds/Fort_Foley_Sounds/Moves/player_run_foley_13}}
 {id=WavePlayer_3, properties {soundWaveAssetPtr=/Game/Sounds/Fort_Foley_Sounds/Moves/player_run_foley_12}}
 {id=WavePlayer_4, properties {soundWaveAssetPtr=&/ExCaliberDMR_ScopeIn}}
]
properties {
 attenuationSettings=&Fort_Weapon_Projectile_FirstPerson_Athena_Att
 bOverrideAttenuation=true
 bOverrideConcurrency=true
 maxDistance=2700
 pitchMultiplier=1.02
 soundClassObject=&Fort_WeaponFoley
 volumeMultiplier=0.1
 attenuationOverrides {falloffDistance=1500, nonSpatializedRadiusStart=200, attenuationShapeExtents {x=1200}}
 concurrencyOverrides {maxCount=1, resolutionRule=StopOldest}
}
