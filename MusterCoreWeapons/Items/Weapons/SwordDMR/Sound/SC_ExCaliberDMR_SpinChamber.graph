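# Sound cue for the chamber-spin foley of the Ex-Caliber DMR weapon, built from open-chamber recordings.
# A Random node picks one of four equally weighted wave variants, feeding a Modulator that adds slight pitch variation.
# The cue raises pitch strongly and boosts volume, using the weapon foley sound class and reload attenuation asset.
# It overrides concurrency and has its debug flag enabled.
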SoundCue /MusterCoreWeapons/Items/Weapons/SwordDMR/Sound/SC_ExCaliberDMR_SpinChamber
root=Modulator_0
nodes [
 {id=Modulator_0, children [Random_1], properties {pitchMax=1.1, pitchMin=0.9}}
 {id=Random_1, children [WavePlayer_7, WavePlayer_8, WavePlayer_9, WavePlayer_10], properties {weights [1,1,1,1]}}
 {id=WavePlayer_10, properties {soundWaveAssetPtr=&/Ex-CaliberDMR_OpenChamber_03}}
 {id=WavePlayer_7, properties {soundWaveAssetPtr=&/Ex-CaliberDMR_OpenChamber_04}}
 {id=WavePlayer_8, properties {soundWaveAssetPtr=&/Ex-CaliberDMR_OpenChamber_01}}
 {id=WavePlayer_9, properties {soundWaveAssetPtr=&/Ex-CaliberDMR_OpenChamber_02}}
]
properties {
 attenuationSettings=&Fort_Weapon_Reload_Att
 bDebug=true
 bOverrideConcurrency=true
 maxDistance=2250
 pitchMultiplier=1.8
 soundClassObject=&Fort_WeaponFoley
 volumeMultiplier=1.7
 attenuationOverrides {falloffDistance=1500, stereoSpread=300, attenuationShapeExtents {x=500}}
}
